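# Subsurface material for the glowing jellyfish in the PartyJelly emote, enabled for Niagara mesh particles.
# Emissive blends blue gradients and masks, scaled by time of day, alongside time-animated Fresnel glows and prop textures.
# World position offset provides local-space sine motion: a head bob, plus stinger sway phased along a UV gradient.
# Vertex color channels and four UV-quadrant masks from the second UV channel are exposed as named reroutes for masking.
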Material /BRCosmetics/Animation/Game/MainPlayer/Emotes/PartyJelly/FX/M_Emote_PartyJelly_Jellyfish
graph {
 properties {
  bAutomaticallySetUsageInEditor=false
  bUsedWithNiagaraMeshParticles=true
  previewMesh=&/SM_Emote_PartyJelly_Jellyfish_01
  shadingModel=MSM_Subsurface
  usedShadingModels=Subsurface
 }
 nodes [
  Multiply(A=NamedRerouteUsage(), B=ScalarParameter() {defaultValue=0.5, parameterName=Specular Adjustment}) {constA=0.5}
  Multiply(A=NamedRerouteUsage()) {constB=0.5}
  n1 = VertexColor()
  NamedRerouteDeclaration(None=@n1.R) {name=Red, nodeColor=<rgba 1 0.0072 0 1>}
  NamedRerouteDeclaration(None=@n1.B) {name=Blue, nodeColor=<rgba 0.0344 0.0596 1 1>}
  n2 = MaterialFunctionCall(Float2=TextureCoordinate() {coordinateIndex=1}) {materialFunction=&BreakOutFloat2Components}
  n3 = NamedRerouteDeclaration(None=Step(Y=@n2.R) {constX=0.5}) {name=UV Mask X, nodeColor=<rgba 1 0.8762 0 1>}
  n4 = NamedRerouteDeclaration(None=Step(Y=@n2.G) {constX=0.5}) {name=UV Mask Y, nodeColor=<rgba 1 0.8438 0 1>}
  NamedRerouteDeclaration(None=Multiply(A=@n3, B=@n4)) {name=UV1, nodeColor=<rgba 0.5234 0 1 1>}
  n5 = OneMinus(None=@n3)
  NamedRerouteDeclaration(None=Multiply(A=@n5, B=@n4)) {name=UV2, nodeColor=<rgba 0 1 0.2031 1>}
  n6 = OneMinus(None=@n4)
  NamedRerouteDeclaration(None=Multiply(A=@n3, B=@n6)) {name=UV3, nodeColor=<rgba 1 0 0.1172 1>}
  NamedRerouteDeclaration(None=Multiply(A=@n5, B=@n6)) {name=UV4, nodeColor=<rgba 0 0.4375 1 1>}
  n7 = MaterialFunctionCall(UV Channel=TextureCoordinate() {coordinateIndex=1}) {materialFunction=&LinearGradient}
  n8 = Multiply(A=NamedRerouteUsage(), B=SmoothStep(Value=@n7.VGradient))
  n9 = SmoothStep(Value=@n7.VGradient) {constMax=0.6, constMin=0.2}
  n10 = NamedRerouteUsage()
  NamedRerouteDeclaration(None=Add(A=Multiply(A=@n8) {constB=2}, B=Multiply(A=@n9, B=@n10))) {name=Gradient Base, nodeColor=<rgba 1 0.7969 0 1>}
  n11 = TextureCoordinate()
  NamedRerouteDeclaration(None=TextureSample(UVs=@n11) {samplerType=SAMPLERTYPE_Masks, texture=/BRCosmetics/Animation/Props/Gadgets_PartyJelly/Textures/T_PartyJelly_Emote_Prop_S}.RGB) {name=Specular, nodeColor=<rgba 1 0 0.0937 1>}
  n12 = Fresnel()
  n13 = Time()
  n14 = LinearInterpolate(
   A=Constant3Vector() {constant=<rgba 0 0.2607 0.5115 1>}
   B=Constant3Vector() {constant=<rgba 0.0001 0.3455 0.5031 1>}
   Alpha=Sine(None=Multiply(A=@n13) {constB=0.6})
  )
  NamedRerouteDeclaration(None=Add(A=Power(Base=@n12) {constExponent=10}, B=Multiply(A=@n12, B=@n14))) {name=Fresnel Glow, nodeColor=<rgba 1 0.5859 0 1>}
  NamedRerouteDeclaration(None=Add(A=NamedRerouteUsage(), B=NamedRerouteUsage())) {name=WPO, nodeColor=<rgba 0 0.4609 1 1>}
  NamedRerouteDeclaration(None=TextureSample(UVs=@n11) {texture=/BRCosmetics/Animation/Props/Gadgets_PartyJelly/Textures/T_PartyJelly_Emote_Prop_E}.RGB) {name=Emissive, nodeColor=<rgba 0 0.4141 1 1>}
  NamedRerouteDeclaration(None=TextureSample(UVs=@n11) {texture=/BRCosmetics/Animation/Props/Gadgets_PartyJelly/Textures/T_PartyJelly_Emote_Prop_D}.RGB) {name=Diffuse, nodeColor=<rgba 0.7344 1 0 1>}
  NamedRerouteDeclaration(None=TextureSample(UVs=@n11) {samplerType=SAMPLERTYPE_Normal, texture=/BRCosmetics/Animation/Props/Gadgets_PartyJelly/Textures/T_PartyJelly_Emote_Prop_N}.RGB) {name=Normal, nodeColor=<rgba 0.9453 0 1 1>}
  n15 = LinearInterpolate(
   A=LinearInterpolate(A=Constant3Vector() {constant=<rgba 0 0.0185 0.074 1>}, B=Constant3Vector() {constant=<rgba 0 0.6272 0.774 1>}, Alpha=NamedRerouteUsage())
   B=Constant3Vector() {constant=<rgba 0 0.6099 0.874 1>}
   Alpha=NamedRerouteUsage()
  )
  n16 = NamedRerouteUsage()
  n17 = NamedRerouteUsage()
  n18 = Add(A=OneMinus(None=@n17), B=Multiply(A=@n17, B=Constant3Vector() {constant=<rgba 0 0.0136 0.0531 1>}))
  n19 = Add(A=LinearInterpolate(A=Add(A=@n15, B=@n16), Alpha=@n17) {constB=0}, B=Multiply(A=Multiply(A=@n17, B=@n18), B=NamedRerouteUsage()))
  n20 = MaterialFunctionCall(
   Value_Morning=ScalarParameter() {defaultValue=1, parameterName=TOD Morning}
   Value_Day=ScalarParameter() {defaultValue=1, parameterName=TOD Day}
   Value_Evening=ScalarParameter() {defaultValue=0.9, parameterName=TOD Evening}
   Value_Night=ScalarParameter() {defaultValue=0.9, parameterName=TOD Night}
  ) {materialFunction=&Scalar_Time_of_Day}
  NamedRerouteDeclaration(None=MaterialFunctionCall(Emissive=Multiply(A=@n19, B=@n20.Result)) {materialFunction=&MF_FX_EmissiveAdjustments}.Result) {name=Emissive Base, nodeColor=<rgba 0.1719 0 1 1>}
  n21 = NamedRerouteUsage()
  n22 = Multiply(A=Time()) {constB=-0.95}
  n23 = Fresnel(ExponentIn=ConstantBiasScale(None=Sine(None=@n22)) {bias=2, scale=0.8}) {baseReflectFraction=0, exponent=1}
  n24 = Multiply(A=Multiply(A=SmoothStep(Value=@n23) {constMax=0.4, constMin=0.2}, B=@n23), B=NamedRerouteUsage())
  n25 = Multiply(A=Power(Base=Multiply(A=@n21, B=@n24)) {constExponent=0.4}) {constB=0.5}
  n26 = Multiply(A=Power(Base=OneMinus(None=@n23)) {constExponent=20}) {constB=0.25}
  NamedRerouteDeclaration(None=Saturate(None=Add(A=@n25, B=@n26))) {name=Glow, nodeColor=<rgba 0.6406 0 1 1>}
  n27 = Transform(None=VectorParameter() {parameterName=Stinger Bob Movement Direction, defaultValue=<rgba 1 1 1 1>}.RGB) {transformSourceType=TRANSFORMSOURCE_Local}
  n28 = Time()
  n29 = ScalarParameter() {defaultValue=1.3, parameterName=Stinger Speed}
  n30 = Power(
   Base=MaterialFunctionCall() {materialFunction=&LinearGradient}.VGradient
   Exp=ScalarParameter() {defaultValue=2, parameterName=Stinger Gradient Power}
  )
  n31 = ScalarParameter() {defaultValue=5, parameterName=Stinger Gradient Intensity}
  n32 = Sine(None=Add(A=Multiply(A=@n28, B=@n29), B=Multiply(A=@n30, B=@n31)))
  n33 = Multiply(
   A=Multiply(A=NamedRerouteUsage(), B=NamedRerouteUsage())
   B=Multiply(A=ScalarParameter() {defaultValue=1, parameterName=Stinger Bob WPO Intensity}, B=@n32)
  )
  n34 = Multiply(
   A=Multiply(A=NamedRerouteUsage(), B=NamedRerouteUsage())
   B=Multiply(A=@n32, B=ScalarParameter() {defaultValue=1.5, parameterName=Stinger Sine WPO Intensity})
  )
  n35 = Transform(None=VectorParameter() {parameterName=Stinger Movement Direction, defaultValue=<rgba 1 1 2 1>}.RGB) {transformSourceType=TRANSFORMSOURCE_Local}
  NamedRerouteDeclaration(None=Add(A=Multiply(A=@n27, B=@n33), B=Multiply(A=@n34, B=@n35))) {name=Stinger WPO, nodeColor=<rgba 0.5469 0 1 1>}
  n36 = Add(A=NamedRerouteUsage(), B=NamedRerouteUsage())
  n37 = NamedRerouteUsage()
  n38 = Multiply(A=@n36, B=@n37)
  n39 = ScalarParameter() {defaultValue=2, parameterName=Blue Intensity}
  n40 = Multiply(A=@n37, B=NamedRerouteUsage())
  n41 = Add(A=Add(A=NamedRerouteUsage(), B=@n36), B=Add(A=Multiply(A=@n38, B=@n39) {constB=1.75}, B=Multiply(A=@n40) {constB=0.65}))
  n42 = Time()
  n43 = ScalarParameter() {defaultValue=0.85, parameterName=Head Speed}
  n44 = Multiply(A=ScalarParameter() {defaultValue=1.4, parameterName=Overall Intensity}, B=Sine(None=Multiply(A=@n42, B=@n43)))
  n45 = Constant3Vector() {constant=<rgba 0 0 1 0>}
  NamedRerouteDeclaration(None=Multiply(A=Multiply(A=@n41, B=@n44), B=Transform(None=@n45) {transformSourceType=TRANSFORMSOURCE_Local})) {name=Head WPO, nodeColor=<rgba 0 1 0.2266 1>}
  NamedRerouteDeclaration(None=@n1.G) {name=Green, nodeColor=<rgba 0 1 0.1533 1>}
  NamedRerouteUsage()
  NamedRerouteUsage()
  NamedRerouteUsage()
  ScalarParameter() {parameterName=Base Color}
  ScalarParameter() {defaultValue=0.2, parameterName=Roughness}
 ]
}
