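# Niagara system for the PartyJelly emote, surrounding the player with short-lived glowing jellyfish and rising bubbles.
# Jellyfish particles spawn on alternating sides, pulsing their glow with waveforms, curves and time-of-day scaling.
# Spawn, despawn and droplet emitters read Jellyfish particles, using age windows to time effects to appearance and disappearance.
# Owner velocity adjusts spawn rates, placement and drag, while front-end modifiers supply separate values outside gameplay.
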NiagaraSystem /BRCosmetics/Animation/Game/MainPlayer/Emotes/PartyJelly/FX/NS_Emote_PartyJelly
dependencies {
 usedDataInterfaces [
  &NiagaraDataInterfaceCurve
  &NiagaraDataInterfaceVectorCurve
  &NiagaraDataInterfaceParticleRead
  &NiagaraDataInterfaceArrayFloat
  &NiagaraDataInterfaceVector2DCurve
  &NiagaraDataInterfaceSkeletalMesh
 ]
 usedDynamicInputs []
 usedModules [
  &SystemState
  &BlendSpaceHelper
  &self:SystemSpawnScript.SystemScriptSource.SystemScriptGraph.NiagaraNodeAssignment_3.SetVariables_#1
 ]
}
emitters {
 AmbientBubbles001 {
  emitterUpdateScript [
   &EmitterState(Life Cycle Mode=::NewEnumerator1, Loop Delay=0.5, Loop Duration=1, UseLoopDelay=false)
   &SpawnRate(
    SpawnRate=&FN_FrontEndFloatModifier(
     Gameplay=&MakeCustomFloatFromBool(
      Bool=&SetBoolByFloatComparison(A=&VectorLength(Vector=$User.OwnerLocalVelocity), B=301)
      True Float=&RandomRangeFloat(Minimum=1.2, Maximum=1.8)
      False Float=&RandomRangeFloat(Minimum=0.6, Maximum=1.2)
     )
     Front End=&RandomRangeFloat(Minimum=1.2, Maximum=1.8)
    )
    Use Spawn Probability=false
   )
   &SpawnRate(
    SpawnGroup=1
    SpawnRate=&FN_FrontEndFloatModifier(
     Gameplay=&MakeCustomFloatFromBool(
      Bool=&SetBoolByFloatComparison(A=&VectorLength(Vector=$User.OwnerLocalVelocity), B=301)
      True Float=&RandomRangeFloat(Minimum=1.2, Maximum=1.8)
      False Float=&RandomRangeFloat(Minimum=0.6, Maximum=1.2)
     )
     Front End=&RandomRangeFloat(Minimum=1.2, Maximum=1.8)
    )
    Use Spawn Probability=false
   ) {module=SpawnRate001}
   &SpawnRate(
    SpawnGroup=2
    SpawnRate=&FN_FrontEndFloatModifier(
     Gameplay=&MakeCustomFloatFromBool(
      Bool=&SetBoolByFloatComparison(A=&VectorLength(Vector=$User.OwnerLocalVelocity), B=301)
      True Float=&RandomRangeFloat(Minimum=1.2, Maximum=1.8)
      False Float=&RandomRangeFloat(Minimum=0.6, Maximum=1.2)
     )
     Front End=&RandomRangeFloat(Minimum=1.2, Maximum=1.8)
    )
    Use Spawn Probability=false
   ) {module=SpawnRate002}
   &SpawnRate(
    SpawnGroup=3
    SpawnRate=&FN_FrontEndFloatModifier(
     Gameplay=&MakeCustomFloatFromBool(
      Bool=&SetBoolByFloatComparison(A=&VectorLength(Vector=$User.OwnerLocalVelocity), B=301)
      True Float=&RandomRangeFloat(Minimum=1.2, Maximum=1.8)
      False Float=&RandomRangeFloat(Minimum=0.6, Maximum=1.2)
     )
     Front End=&RandomRangeFloat(Minimum=1.2, Maximum=1.8)
    )
    Use Spawn Probability=false
   ) {module=SpawnRate003}
  ]
  particleSpawnScript [
   &InitializeParticle(
    Lifetime=&RandomRangeFloat(Minimum=1.3, Maximum=4.3)
    Sprite Size Mode=::NewEnumerator1
    Uniform Sprite Size Max=3
    Uniform Sprite Size Min=1
    UsePositionOffset=false
    Write Material Random=true
   )
   &AddVelocity(Velocity=&MaskVector(
     Mask Source=0
     Mask Parameter=?
     Mask=0
     Masked Vector=&RandomRangeVector(Minimum=<-40 -50 30>, Maximum=<10 10 120>)
     Passthrough Vector=&MaskVector(
      Mask Source=0
      Mask Parameter=?
      Mask=1
      Masked Vector=&RandomRangeVector(Minimum=<-10 -50 30>, Maximum=<40 10 120>)
      Passthrough Vector=&MaskVector(
       Mask Source=0
       Mask Parameter=?
       Mask=2
       Masked Vector=&RandomRangeVector(Minimum=<-10 -70 30>, Maximum=<40 -90 60>)
       Passthrough Vector=&RandomRangeVector(Minimum=<-10 -70 30>, Maximum=<40 -90 60>)
      )
     )
    ))
   &ShapeLocation(
    Box Size=&MaskVector(
     Mask Source=0
     Mask Parameter=?
     Mask=0
     Masked Vector=<10 150 50>
     Passthrough Vector=&MaskVector(
      Mask Source=0
      Mask Parameter=?
      Mask=1
      Masked Vector=<10 150 50>
      Passthrough Vector=&MaskVector(Mask Source=0, Mask Parameter=?, Mask=2, Masked Vector=<10 200 65>, Passthrough Vector=<10 200 65>)
     )
    )
    MaskSpawnGroup=false
    Offset=&MaskVector(
     Mask Source=0
     Mask Parameter=?
     Mask=0
     Masked Vector=<-50 -25 40>
     Passthrough Vector=&MaskVector(
      Mask Source=0
      Mask Parameter=?
      Mask=1
      Masked Vector=<55 -25 40>
      Passthrough Vector=&MaskVector(Mask Source=0, Mask Parameter=?, Mask=2, Masked Vector=<-50 -10 50>, Passthrough Vector=<55 -10 50>)
     )
    )
    Offset Mode=::NewEnumerator0
    Override Seed=false
    Shape Primitive=::NewEnumerator2
    Surface Only=false
    UseOwnerScale=true
   )
   &ApplyInitialForces()
   &AddVelocity(Velocity=&FN_FrontEndVector3Modifier(Gameplay=<0 0 0>, Front End=<0 -15 0>)) {module=AddVelocity001}
  ]
  particleUpdateScript [
   &ParticleState()
   &Timeline(
    Play=&SetBoolByFloatComparison(
     A=&RemapRange(
      Input Value To Remap=&VectorLength(Vector=$User.OwnerLocalVelocity)
      Input Min=1
      Input Max=300
      Output Min=0
      Output Max=1
     )
     B=0.5
    )
    Play Rate=5
   )
   &InheritVelocity(
    Inherited Velocity Amount Scale=<-1 -1 0>
    Inherited Velocity Speed Limit=10
    Speed Limit=true
    UseSpeedThreshold=false
    Velocity Source=$User.OwnerLocalVelocity
   )
   &CurlNoiseForce(
    Noise Strength=&FloatFromCurve(FloatCurve=DataInterface() {propertyValues {Curve {keys [
         {interpMode=RCIM_Cubic, leaveTangentWeight=53.6425, tangentMode=RCTM_User, time=0, value=50}
         {arriveTangent=342.4859, interpMode=RCIM_Cubic, leaveTangent=342.4859, tangentMode=RCTM_User, time=1, value=150}
        ]}}})
    Pan Noise=false
   )
   &GravityForce(Gravity=&RandomRangeVector(
     Minimum=&VectorFromCurve(VectorCurve=DataInterface() {propertyValues {
        XCurve {keys [<at 0 0>,<at 1 0>]}
        YCurve {keys [<at 0 0>,<at 1 0>]}
        ZCurve {keys [
          {interpMode=RCIM_Cubic, tangentMode=RCTM_User, time=0, value=10}
          {arriveTangent=88.4209, interpMode=RCIM_Cubic, leaveTangent=88.4209, tangentMode=RCTM_User, time=0.75, value=35}
         ]}
       }})
     Maximum=&VectorFromCurve(VectorCurve=DataInterface() {propertyValues {
        XCurve {keys [<at 0 0>,<at 1 0>]}
        YCurve {keys [<at 0 0>,<at 1 0>]}
        ZCurve {keys [
          {interpMode=RCIM_Cubic, tangentMode=RCTM_User, time=0, value=10}
          {arriveTangent=133.5137, interpMode=RCIM_Cubic, leaveTangent=133.5137, tangentMode=RCTM_User, time=0.75, value=70}
         ]}
       }})
    ))
   &DynamicMaterialParameters(
    Index 0 Param 1=&FloatFromCurve(FloatCurve=DataInterface() {propertyValues {Curve {keys [
         {arriveTangent=0.1706, interpMode=RCIM_Cubic, leaveTangent=0.1706, leaveTangentWeight=0.3081, tangentMode=RCTM_User, time=0, value=0}
         {arriveTangentWeight=0.3104, interpMode=RCIM_Cubic, leaveTangentWeight=0.0345, tangentMode=RCTM_User, time=0.9, value=0.2}
         {arriveTangent=3.2005, interpMode=RCIM_Cubic, leaveTangent=3.2005, tangentMode=RCTM_User, time=1, value=0.5}
        ]}}})
    Module.Index0Float0Write=true
    Module.Index0Float1Write=true
    Module.Index0Float2Write=true
    Module.Index0Float3Write=true
   )
   &AccelerationForce(Acceleration=&MaskVector(
     Mask Source=0
     Mask Parameter=?
     Mask=0
     Masked Vector=&VectorFromCurve(VectorCurve=DataInterface() {propertyValues {
        XCurve {keys [
          {arriveTangent=-9302.5869, interpMode=RCIM_Cubic, leaveTangent=-9302.5869, tangentMode=RCTM_User, time=0, value=0}
          {interpMode=RCIM_Cubic, leaveTangent=429.3502, tangentMode=RCTM_Break, time=0.05, value=-150}
          {interpMode=RCIM_Cubic, tangentMode=RCTM_User, time=0.6, value=0}
         ]}
        YCurve {keys [<at 0 -25>,<at 1 -62.5>]}
        ZCurve {keys [
          {interpMode=RCIM_Cubic, leaveTangent=-447.6769, tangentMode=RCTM_Break, time=0.05, value=90}
          {interpMode=RCIM_Cubic, tangentMode=RCTM_User, time=0.6, value=0}
         ]}
       }})
     Passthrough Vector=&MaskVector(
      Mask Source=0
      Mask Parameter=?
      Mask=1
      Masked Vector=&VectorFromCurve(VectorCurve=DataInterface() {propertyValues {
         XCurve {
          defaultValue=0
          keys [
           {arriveTangent=3478.5315, interpMode=RCIM_Cubic, leaveTangent=3478.5315, tangentMode=RCTM_User, time=0, value=0}
           {interpMode=RCIM_Cubic, leaveTangent=-700.6756, tangentMode=RCTM_Break, time=0.1, value=150}
           {interpMode=RCIM_Cubic, tangentMode=RCTM_User, time=0.6, value=0}
          ]
         }
         YCurve {keys [<at 0 -25>,<at 1 -40>]}
         ZCurve {
          defaultValue=0
          keys [
           {interpMode=RCIM_Cubic, leaveTangent=-447.6769, leaveTangentWeight=0.1833, tangentMode=RCTM_Break, time=0.05, value=90}
           {arriveTangentWeight=0.1833, interpMode=RCIM_Cubic, tangentMode=RCTM_User, time=0.6, value=0}
          ]
         }
        }})
      Passthrough Vector=&MaskVector(
       Mask Source=0
       Mask Parameter=?
       Mask=2
       Masked Vector=&VectorFromCurve(VectorCurve=DataInterface() {propertyValues {
          XCurve {keys [
            {arriveTangent=-9302.5869, interpMode=RCIM_Cubic, leaveTangent=-9302.5869, tangentMode=RCTM_User, time=0, value=0}
            {interpMode=RCIM_Cubic, leaveTangent=429.3502, tangentMode=RCTM_Break, time=0.05, value=-150}
            {interpMode=RCIM_Cubic, tangentMode=RCTM_User, time=0.6, value=0}
           ]}
          YCurve {keys [<at 0 -25>,<at 1 -62.5>]}
          ZCurve {keys [
            {interpMode=RCIM_Cubic, leaveTangent=-447.6769, tangentMode=RCTM_Break, time=0.05, value=90}
            {interpMode=RCIM_Cubic, tangentMode=RCTM_User, time=0.6, value=0}
           ]}
         }})
       Passthrough Vector=&VectorFromCurve(VectorCurve=DataInterface() {propertyValues {
          XCurve {
           defaultValue=0
           keys [
            {arriveTangent=3478.5315, interpMode=RCIM_Cubic, leaveTangent=3478.5315, tangentMode=RCTM_User, time=0, value=0}
            {interpMode=RCIM_Cubic, leaveTangent=-700.6756, tangentMode=RCTM_Break, time=0.1, value=150}
            {interpMode=RCIM_Cubic, tangentMode=RCTM_User, time=0.6, value=0}
           ]
          }
          YCurve {keys [<at 0 -25>,<at 1 -40>]}
          ZCurve {
           defaultValue=0
           keys [
            {interpMode=RCIM_Cubic, leaveTangent=-447.6769, leaveTangentWeight=0.1833, tangentMode=RCTM_Break, time=0.05, value=90}
            {arriveTangentWeight=0.1833, interpMode=RCIM_Cubic, tangentMode=RCTM_User, time=0.6, value=0}
           ]
          }
         }})
      )
     )
    ))
   &ScaleColor(
    Scale Alpha=&FloatFromCurve(
     FloatCurve=DataInterface() {propertyValues {Curve {keys [
         {interpMode=RCIM_Cubic, tangentMode=RCTM_User, time=0, value=0.3}
         {arriveTangent=-0.7622, interpMode=RCIM_Cubic, leaveTangent=-0.7622, tangentMode=RCTM_User, time=0.26, value=0.2}
         {arriveTangent=-0.2703, interpMode=RCIM_Cubic, leaveTangent=-0.2703, time=0.5, value=0.08}
         {interpMode=RCIM_Cubic, tangentMode=RCTM_User, time=1, value=0}
        ]}}}
     Scale Curve=0.5
    )
    ScaleA=true
    ScaleRGB=true
   )
   &ParticleState() {module=ParticleState001}
   &ScaleSpriteSize(Scale Factor=&Vector2DFromFloat(Value=&FloatFromCurve(
      FloatCurve=DataInterface() {propertyValues {Curve {keys [
          {arriveTangent=4.1785, interpMode=RCIM_Cubic, leaveTangent=4.1785, tangentMode=RCTM_User, time=0, value=0}
          {arriveTangent=1.8588, interpMode=RCIM_Cubic, leaveTangent=1.8588, tangentMode=RCTM_User, time=0.1, value=0.35}
          {arriveTangent=-0.2915, interpMode=RCIM_Cubic, leaveTangent=-0.2915, tangentMode=RCTM_User, time=0.6, value=1}
          <at 1 0 RCIM_Cubic>
         ]}}}
      Scale Curve=4
     )))
   &SpriteRotationRate(Rotation Rate=&RandomRangeFloat(Minimum=-40, Maximum=40))
   &Drag(
    Drag=&MakeCustomFloatFromBool(
     Bool=&SetBoolByFloatComparison(A=&VectorLength(Vector=$User.OwnerLocalVelocity), B=150)
     True Float=&RandomRangeFloat(Minimum=4, Maximum=10)
     False Float=&RandomRangeFloat(Minimum=2, Maximum=5)
    )
    Use Linear Drag=true
    Use Rotational Drag=true
   )
   &InheritSourceMovement(
    Applied Position Scale=&VectorFromCurve(VectorCurve=DataInterface() {propertyValues {
       XCurve {keys [<at 0 0>,<at 0.7 0>,<at 1 -0.2>]}
       YCurve {keys [<at 0 0>,<at 0.7 0>,<at 1 -0.2>]}
       ZCurve {keys [<at 0 0>,<at 0.7 0>,<at 1 -0.2>]}
      }})
    ApplyForce=false
    ApplyPosition=true
    ApplyPositionFalloff=false
   )
   &SolveForcesAndVelocity(Clamp Velocity=false, Limit Acceleration=false)
  ]
  rendererClasses [&NiagaraSpriteRendererProperties]
  renderers [{rendererClass=&NiagaraSpriteRendererProperties}]
 }
 BubbleBurst {
  emitterUpdateScript [
   &EmitterState(Life Cycle Mode=::NewEnumerator1, Loop Behavior=::NewEnumerator1, Loop Duration=2, UseLoopDelay=false)
   &SpawnBurst_Instantaneous(
    Spawn Count=&RandomRangeInt(Minimum=2, Maximum=3)
    Spawn Time=0.14
    Use Spawn Probability=false
    UseLoopCountLimit=false
   )
   &SpawnBurst_Instantaneous(
    Spawn Count=&RandomRangeInt(Minimum=2, Maximum=3)
    Spawn Group=1
    Spawn Time=0.14
    Use Spawn Probability=false
    UseLoopCountLimit=false
   ) {module=SpawnBurst_Instantaneous001}
  ]
  particleSpawnScript [
   &InitializeParticle(
    Lifetime=&RandomRangeFloat(Minimum=2, Maximum=4)
    Sprite Size Mode=::NewEnumerator1
    Uniform Sprite Size Max=5
    Uniform Sprite Size Min=1
    UsePositionOffset=false
    Write Material Random=true
   )
   &AddVelocity(Velocity=&MaskVector(
     Mask Source=0
     Mask Parameter=?
     Mask=0
     Masked Vector=&RandomRangeVector(Minimum=<-40 -10 30>, Maximum=<10 10 120>)
     Passthrough Vector=&RandomRangeVector(Minimum=<-10 -10 30>, Maximum=<40 10 120>)
    ))
   &ShapeLocation(
    Box Size=<10 100 50>
    MaskSpawnGroup=false
    Offset=&MaskVector(Mask Source=0, Mask Parameter=?, Mask=0, Masked Vector=<-50 -8 40>, Passthrough Vector=<55 -8 40>)
    Offset Mode=::NewEnumerator0
    Override Seed=false
    Shape Primitive=::NewEnumerator2
    Surface Only=false
    UseOwnerScale=true
   )
   &ApplyInitialForces()
   &AddVelocity() {module=AddVelocity001}
  ]
  particleUpdateScript [
   &ParticleState()
   &InheritVelocity(
    Inherited Velocity Amount Scale=<-1 -1 0>
    Inherited Velocity Speed Limit=10
    Speed Limit=true
    UseSpeedThreshold=false
    Velocity Source=$User.OwnerLocalVelocity
   )
   &CurlNoiseForce(
    Noise Strength=&FloatFromCurve(FloatCurve=DataInterface() {propertyValues {Curve {keys [
         {interpMode=RCIM_Cubic, tangentMode=RCTM_User, time=0, value=80}
         {arriveTangent=772.9872, interpMode=RCIM_Cubic, leaveTangent=772.9872, tangentMode=RCTM_User, time=1, value=350}
        ]}}})
    Pan Noise=false
   )
   &GravityForce(Gravity=&RandomRangeVector(
     Minimum=&VectorFromCurve(VectorCurve=DataInterface() {propertyValues {
        XCurve {keys [<at 0 0>,<at 1 0>]}
        YCurve {keys [<at 0 0>,<at 1 0>]}
        ZCurve {keys [
          {interpMode=RCIM_Cubic, tangentMode=RCTM_User, time=0, value=10}
          {arriveTangent=88.4209, interpMode=RCIM_Cubic, leaveTangent=88.4209, tangentMode=RCTM_User, time=0.75, value=35}
         ]}
       }})
     Maximum=&VectorFromCurve(VectorCurve=DataInterface() {propertyValues {
        XCurve {keys [<at 0 0>,<at 1 0>]}
        YCurve {keys [<at 0 0>,<at 1 0>]}
        ZCurve {keys [
          {interpMode=RCIM_Cubic, tangentMode=RCTM_User, time=0, value=10}
          {arriveTangent=133.5137, interpMode=RCIM_Cubic, leaveTangent=133.5137, tangentMode=RCTM_User, time=0.75, value=70}
         ]}
       }})
    ))
   &DynamicMaterialParameters(
    Index 0 Param 1=&FloatFromCurve(FloatCurve=DataInterface() {propertyValues {Curve {keys [
         {arriveTangent=0.2342, interpMode=RCIM_Cubic, leaveTangent=0.2342, tangentMode=RCTM_User, time=0, value=0.2}
         {arriveTangent=0.2651, interpMode=RCIM_Cubic, leaveTangent=0.2651, tangentMode=RCTM_User, time=0.9, value=0.25}
         {arriveTangent=3.2005, interpMode=RCIM_Cubic, leaveTangent=3.2005, tangentMode=RCTM_User, time=1, value=0.75}
        ]}}})
    Module.Index0Float0Write=true
    Module.Index0Float1Write=true
    Module.Index0Float2Write=true
    Module.Index0Float3Write=true
   )
   &AccelerationForce(Acceleration=&MaskVector(
     Mask Source=0
     Mask Parameter=?
     Mask=0
     Masked Vector=&VectorFromCurve(VectorCurve=DataInterface() {propertyValues {
        XCurve {keys [
          {arriveTangent=-9302.5869, interpMode=RCIM_Cubic, leaveTangent=-9302.5869, tangentMode=RCTM_User, time=0, value=0}
          {interpMode=RCIM_Cubic, leaveTangent=429.3502, tangentMode=RCTM_Break, time=0.05, value=-150}
          {interpMode=RCIM_Cubic, tangentMode=RCTM_User, time=0.6, value=0}
         ]}
        YCurve {keys [<at 0 0>,<at 1 0>]}
        ZCurve {keys [
          {interpMode=RCIM_Cubic, leaveTangent=-447.6769, tangentMode=RCTM_Break, time=0.05, value=90}
          {interpMode=RCIM_Cubic, tangentMode=RCTM_User, time=0.6, value=0}
         ]}
       }})
     Passthrough Vector=&VectorFromCurve(VectorCurve=DataInterface() {propertyValues {
        XCurve {
         defaultValue=0
         keys [
          {arriveTangent=3478.5315, interpMode=RCIM_Cubic, leaveTangent=3478.5315, tangentMode=RCTM_User, time=0, value=0}
          {interpMode=RCIM_Cubic, leaveTangent=-700.6756, tangentMode=RCTM_Break, time=0.1, value=150}
          {interpMode=RCIM_Cubic, tangentMode=RCTM_User, time=0.6, value=0}
         ]
        }
        YCurve {keys [<at 0 0>,<at 1 0>]}
        ZCurve {
         defaultValue=0
         keys [
          {interpMode=RCIM_Cubic, leaveTangent=-447.6769, leaveTangentWeight=0.1833, tangentMode=RCTM_Break, time=0.05, value=90}
          {arriveTangentWeight=0.1833, interpMode=RCIM_Cubic, tangentMode=RCTM_User, time=0.6, value=0}
         ]
        }
       }})
    ))
   &ScaleColor(
    Scale Alpha=&FloatFromCurve(FloatCurve=DataInterface() {propertyValues {Curve {keys [
         {interpMode=RCIM_Cubic, leaveTangentWeight=0.1216, tangentMode=RCTM_User, time=0, value=0.95}
         {arriveTangent=-2.4777, interpMode=RCIM_Cubic, leaveTangent=-2.4777, tangentMode=RCTM_User, time=0.26, value=0.75}
         {arriveTangent=-2.274, interpMode=RCIM_Cubic, leaveTangent=-2.274, tangentMode=RCTM_User, time=0.4, value=0.25}
         {arriveTangent=-0.4872, interpMode=RCIM_Cubic, leaveTangent=-0.4872, tangentMode=RCTM_User, time=0.6, value=0.08}
         {arriveTangentWeight=0.1871, interpMode=RCIM_Cubic, tangentMode=RCTM_User, time=1, value=0}
        ]}}})
    ScaleA=true
    ScaleRGB=true
   )
   &ParticleState() {module=ParticleState001}
   &ScaleSpriteSize(Scale Factor=&Vector2DFromFloat(Value=&FloatFromCurve(
      FloatCurve=DataInterface() {propertyValues {Curve {keys [
          {arriveTangent=4.1785, interpMode=RCIM_Cubic, leaveTangent=4.1785, tangentMode=RCTM_User, time=0, value=0}
          {arriveTangent=0.9995, interpMode=RCIM_Cubic, leaveTangent=0.9995, tangentMode=RCTM_User, time=0.1, value=0.35}
          <at 0.9 0.75>
          {arriveTangent=2.5, interpMode=RCIM_Cubic, leaveTangent=2.5, time=0.95, value=0.57}
          {arriveTangent=4.4385, interpMode=RCIM_Cubic, leaveTangent=4.4385, tangentMode=RCTM_User, time=1, value=1}
         ]}}}
      Scale Curve=4
     )))
   &SpriteRotationRate(Rotation Rate=&RandomRangeFloat(Minimum=-40, Maximum=40))
   &Drag(
    Drag=&FN_FrontEndFloatModifier(
     Gameplay=&MakeCustomFloatFromBool(
      Bool=&SetBoolByFloatComparison(A=&VectorLength(Vector=$User.OwnerLocalVelocity), B=150)
      True Float=&RandomRangeFloat(Minimum=2, Maximum=6)
      False Float=&RandomRangeFloat(Minimum=0.5, Maximum=2)
     )
     Front End=&RandomRangeFloat(Minimum=2, Maximum=3)
    )
    Use Linear Drag=true
    Use Rotational Drag=true
   )
   &SolveForcesAndVelocity(Clamp Velocity=false, Limit Acceleration=false)
  ]
  rendererClasses [&NiagaraSpriteRendererProperties]
  renderers [{rendererClass=&NiagaraSpriteRendererProperties}]
 }
 Bubbles_Despawn {
  emitterUpdateScript [
   &EmitterState()
   &SpawnParticlesFromOtherEmitter(
    Attribute Reader=DataInterface() {propertyValues {EmitterBinding {bindingMode=Other, emitterName=Jellyfish}}}
    Spawn Rate=&RandomRangeFloat(Minimum=35, Maximum=40)
    Use Maximum Age=false
    Use Minimum Age=false
   )
  ]
  particleSpawnScript [
   &InitializeParticle(
    Color Mode=::NewEnumerator0
    Lifetime=0.5
    Sprite Size Mode=::NewEnumerator1
    Uniform Sprite Size Max=8
    Uniform Sprite Size Min=1
    UsePositionOffset=false
    Write Material Random=true
   )
   &SampleParticlesFromOtherEmitter(
    Maximum Sampled Particle Age=3.32
    Minimum Sampled Particle Age=3.25
    Module.MaxSampledAgeCutoff=true
    Module.MinSampledAgeCutoff=true
    Reject Particle By Sampled Age=true
   )
   &AddVelocity(
    Module.Use Falloff Distance=false
    Override Seed=false
    Velocity Mode=::NewEnumerator1
    Velocity Speed=&RandomRangeFloat(Minimum=60, Maximum=100)
   )
   &ShapeLocation(
    MaskSpawnGroup=false
    Non Uniform Scale=<1 0 1.8>
    Offset Mode=::NewEnumerator0
    Override Seed=false
    Sphere Radius=15
    Sphere Surface Distribution=0.6
    UseOwnerScale=true
   )
   &ApplyInitialForces()
   &CameraOffset(Camera Offset Amount=&RandomRangeFloat(Minimum=-5, Maximum=15))
  ]
  particleUpdateScript [
   &ParticleState()
   &ScaleSpriteSize(Uniform Curve Sprite Scale=DataInterface() {propertyValues {Curve {keys [<at 0 0 RCIM_Cubic>,<at 0.7 1 RCIM_Cubic>,<at 1 0 RCIM_Cubic>]}}})
   &DynamicMaterialParameters(
    Index 0 Param 1=&FloatFromCurve(FloatCurve=DataInterface() {propertyValues {Curve {keys [
         {arriveTangent=0.2342, interpMode=RCIM_Cubic, leaveTangent=0.2342, tangentMode=RCTM_User, time=0, value=0.2}
         {arriveTangent=0.2651, interpMode=RCIM_Cubic, leaveTangent=0.2651, tangentMode=RCTM_User, time=0.9, value=0.25}
         {arriveTangent=3.2005, interpMode=RCIM_Cubic, leaveTangent=3.2005, tangentMode=RCTM_User, time=1, value=0.75}
        ]}}})
    Module.Index0Float0Write=true
    Module.Index0Float1Write=true
    Module.Index0Float2Write=true
    Module.Index0Float3Write=true
   )
   &ScaleColor(
    Scale Alpha=&FloatFromCurve(FloatCurve=DataInterface() {propertyValues {Curve {keys [
         {arriveTangent=-1.406, interpMode=RCIM_Cubic, leaveTangent=-1.406, leaveTangentWeight=0.1216, tangentMode=RCTM_User, time=0, value=0.1}
         {arriveTangent=-1.4929, interpMode=RCIM_Cubic, leaveTangent=-1.4929, tangentMode=RCTM_User, time=0.3, value=0.3}
         {arriveTangentWeight=0.1871, interpMode=RCIM_Cubic, tangentMode=RCTM_User, time=1, value=0}
        ]}}})
    ScaleA=true
    ScaleRGB=true
   ) {module=ScaleColor001}
   &GravityForce(Gravity=&VectorFromCurve(VectorCurve=DataInterface() {propertyValues {
       XCurve {keys [<at 0 0>,<at 1 0>]}
       YCurve {keys [<at 0 0>,<at 1 0>]}
       ZCurve {keys [
         {interpMode=RCIM_Cubic, tangentMode=RCTM_User, time=0, value=0}
         {arriveTangent=8.1272, interpMode=RCIM_Cubic, leaveTangent=8.1272, tangentMode=RCTM_User, time=0.75, value=2}
        ]}
      }}))
   &CurlNoiseForce(Noise Strength=250, Pan Noise=false)
   &SolveForcesAndVelocity(Clamp Velocity=false, Limit Acceleration=false)
  ]
  rendererClasses [&NiagaraSpriteRendererProperties]
  renderers [{rendererClass=&NiagaraSpriteRendererProperties}]
 }
 Bubbles_Spawn {
  emitterUpdateScript [
   &EmitterState()
   &SpawnParticlesFromOtherEmitter(
    Attribute Reader=DataInterface() {propertyValues {EmitterBinding {bindingMode=Other, emitterName=Jellyfish}}}
    Maximum Age Spawn Threshold=0.5
    Spawn Rate=&RandomRangeFloat(Minimum=25, Maximum=30)
    Use Maximum Age=false
    Use Minimum Age=false
   )
  ]
  particleSpawnScript [
   &InitializeParticle(
    Color Mode=::NewEnumerator0
    Lifetime=0.5
    Sprite Size Mode=::NewEnumerator1
    Uniform Sprite Size Max=7
    Uniform Sprite Size Min=1
    UsePositionOffset=false
    Write Material Random=true
   )
   &SampleParticlesFromOtherEmitter(
    Maximum Sampled Particle Age=0.08
    Minimum Sampled Particle Age=0.1
    Module.MaxSampledAgeCutoff=true
    Module.MinSampledAgeCutoff=false
    Reject Particle By Sampled Age=true
   )
   &AddVelocity(
    Module.Use Falloff Distance=false
    Origin Offset=<0 0 -10>
    Override Seed=false
    Velocity Mode=::NewEnumerator1
    Velocity Speed=&RandomRangeFloat(Minimum=70, Maximum=100)
   )
   &ShapeLocation(
    MaskSpawnGroup=false
    Non Uniform Scale=<1 0 2>
    Offset=<0 0 -2>
    Offset Mode=::NewEnumerator0
    Override Seed=false
    Sphere Radius=12
    Sphere Surface Distribution=0.5
    UseOwnerScale=true
   )
   &ApplyInitialForces()
   &CameraOffset(Camera Offset Amount=&RandomRangeFloat(Minimum=15, Maximum=25))
  ]
  particleUpdateScript [
   &ParticleState()
   &DynamicMaterialParameters(
    Index 0 Param 1=&FloatFromCurve(FloatCurve=DataInterface() {propertyValues {Curve {keys [
         {arriveTangent=0.2342, interpMode=RCIM_Cubic, leaveTangent=0.2342, tangentMode=RCTM_User, time=0, value=0.2}
         {arriveTangent=0.2651, interpMode=RCIM_Cubic, leaveTangent=0.2651, tangentMode=RCTM_User, time=0.9, value=0.25}
         {arriveTangent=3.2005, interpMode=RCIM_Cubic, leaveTangent=3.2005, tangentMode=RCTM_User, time=1, value=0.75}
        ]}}})
    Module.Index0Float0Write=true
    Module.Index0Float1Write=true
    Module.Index0Float2Write=true
    Module.Index0Float3Write=true
   )
   &ScaleColor(
    Scale Alpha=&FloatFromCurve(FloatCurve=DataInterface() {propertyValues {Curve {keys [
         {interpMode=RCIM_Cubic, leaveTangentWeight=0.1216, tangentMode=RCTM_User, time=0, value=0.7}
         {arriveTangent=-2.4777, interpMode=RCIM_Cubic, leaveTangent=-2.4777, tangentMode=RCTM_User, time=0.26, value=0.4}
         {arriveTangent=-0.8512, interpMode=RCIM_Cubic, leaveTangent=-0.8512, tangentMode=RCTM_User, time=0.4, value=0.1}
         {arriveTangent=-0.4872, interpMode=RCIM_Cubic, leaveTangent=-0.4872, tangentMode=RCTM_User, time=0.6, value=0.08}
         {arriveTangentWeight=0.1871, interpMode=RCIM_Cubic, tangentMode=RCTM_User, time=1, value=0}
        ]}}})
    ScaleA=true
    ScaleRGB=true
   )
   &GravityForce(Gravity=&VectorFromCurve(VectorCurve=DataInterface() {propertyValues {
       XCurve {keys [<at 0 0>,<at 1 0>]}
       YCurve {keys [<at 0 0>,<at 1 0>]}
       ZCurve {keys [
         {interpMode=RCIM_Cubic, tangentMode=RCTM_User, time=0, value=0}
         {arriveTangent=21.1292, interpMode=RCIM_Cubic, leaveTangent=21.1292, tangentMode=RCTM_User, time=0.75, value=5}
        ]}
      }}))
   &CurlNoiseForce(Noise Strength=250, Pan Noise=false)
   &SolveForcesAndVelocity(Clamp Velocity=false, Limit Acceleration=false)
  ]
  rendererClasses [&NiagaraSpriteRendererProperties]
  renderers [{rendererClass=&NiagaraSpriteRendererProperties}]
 }
 Droplets {
  emitterUpdateScript [
   &EmitterState(Life Cycle Mode=::NewEnumerator1, Loop Duration=0, UseLoopDelay=false)
   &SpawnParticlesFromOtherEmitter(
    Attribute Reader=DataInterface() {propertyValues {EmitterBinding {bindingMode=Other, emitterName=Jellyfish}}}
    Spawn Rate=&RandomRangeFloat(Minimum=50, Maximum=55)
    Use Maximum Age=false
    Use Minimum Age=false
   )
  ]
  particleSpawnScript [
   &InitializeParticle(
    Color=<rgba 0.2031 0.7612 1 0.35>
    Lifetime Max=0.33
    Lifetime Min=0.23
    Lifetime Mode=::NewEnumerator1
    Sprite Size Max=<2 3.25>
    Sprite Size Min=<1.5 2.75>
    Sprite Size Mode=::NewEnumerator2
    UsePositionOffset=false
    Write Material Random=true
   )
   &AddVelocityFromPoint(Module.Use Falloff Distance=false)
   &SampleParticlesFromOtherEmitter(
    Maximum Sampled Particle Age=3.42
    Minimum Sampled Particle Age=3.35
    Module.MaxSampledAgeCutoff=true
    Module.MinSampledAgeCutoff=true
    Reject Particle By Sampled Age=true
   )
   &ShapeLocation(MaskSpawnGroup=false, Offset=<0 0 8>, Offset Mode=::NewEnumerator0, Override Seed=false, Sphere Radius=8, UseOwnerScale=true)
  ]
  particleUpdateScript [
   &ParticleState()
   &ScaleSpriteSize(
    Non-Uniform Curve Sprite Scale=DataInterface() {propertyValues {
      XCurve {keys [{tangentMode=RCTM_User, time=0, value=1}, {tangentMode=RCTM_User, time=1, value=1}]}
      YCurve {keys [
        {arriveTangent=3.8953, interpMode=RCIM_Cubic, leaveTangent=3.8953, tangentMode=RCTM_User, time=0, value=1}
        {interpMode=RCIM_Cubic, tangentMode=RCTM_User, time=1, value=2}
       ]}
     }}
    Scale Sprite Size Mode=::NewEnumerator3
   ) {module=ScaleSpriteSize001}
   &CurlNoiseForce(Noise Strength=250, Pan Noise=false)
   &ScaleColor(
    Scale Alpha=&FloatFromCurve(FloatCurve=DataInterface() {propertyValues {Curve {keys [
         {arriveTangent=1.8, interpMode=RCIM_Cubic, leaveTangent=1.8, time=0, value=0}
         {arriveTangent=0.002, interpMode=RCIM_Cubic, leaveTangent=0.002, time=0.5, value=0.9}
         {arriveTangent=-1.792, interpMode=RCIM_Cubic, leaveTangent=-1.792, time=1.0011, value=0.002}
        ]}}})
    ScaleA=true
    ScaleRGB=true
   )
   &Drag(Drag=&RandomRangeFloat(Minimum=0.5, Maximum=1), Use Linear Drag=true, Use Rotational Drag=false)
   &GravityForce(Gravity=&RandomRangeVector(Minimum=<0 0 -128>, Maximum=<0 0 -256>))
   &AccelerationForce(Acceleration=<0 0 -70>)
   &SolveForcesAndVelocity(Clamp Velocity=false, Limit Acceleration=false)
  ]
  rendererClasses [&NiagaraSpriteRendererProperties]
  renderers [{rendererClass=&NiagaraSpriteRendererProperties}]
 }
 Jellyfish {
  emitterUpdateScript [
   &EmitterState(Life Cycle Mode=::NewEnumerator1, Loop Delay=0.1, Loop Duration=1, UseLoopDelay=false)
   &SpawnBurst_Instantaneous(Use Spawn Probability=false, UseLoopCountLimit=true)
   &SpawnRate(SpawnRate=1.3, Use Spawn Probability=false)
  ]
  particleSpawnScript [
   &PartitionParticles()
   &InitializeParticle(
    Color=<rgba 0.112 0.635 1 0.28>
    Lifetime=3.55
    Mesh Scale Mode=::NewEnumerator1
    Mesh Uniform Scale Max=0.95
    Mesh Uniform Scale Min=0.65
    Sprite Size Mode=::NewEnumerator1
    Uniform Sprite Size Max=85
    Uniform Sprite Size Min=75
    UsePositionOffset=false
    Write Material Random=true
   )
   &ShapeLocation(
    Box Size=&FN_FrontEndVector3Modifier(
     Gameplay=&MakeCustomVectorFromBool(
      Bool=&SetBoolByFloatComparison(A=&VectorLength(Vector=$User.OwnerLocalVelocity), B=50)
      True Vector=<50 50 60>
      False Vector=<50 200 50>
     )
     Front End=<25 200 50>
    )
    MaskSpawnGroup=false
    Offset=&FN_FrontEndVector3Modifier(
     Gameplay=&MakeCustomVectorFromBool(
      Bool=&SetBoolByFloatComparison(A=&VectorLength(Vector=$User.OwnerLocalVelocity), B=50)
      True Vector=&LerpPosition(
       A=<100 165 75>
       B=<-100 155 75>
       Alpha=&Lerp_Float(
        A=&RandomRangeFloat(Minimum=0, Maximum=0.03)
        B=&RandomRangeFloat(Minimum=0.97, Maximum=1)
        Alpha=&SelectFloatFromArray(
         Direct Array Index=&ModuloInt(A=$Engine.Emitter.TotalSpawnedParticles, B=2)
         Float Selection Array=DataInterface() {propertyValues {FloatData [0,1]}}
         Random Seed=0
        )
       )
      )
      False Vector=&LerpPosition(
       A=<100 55 60>
       B=<-100 55 60>
       Alpha=&Lerp_Float(
        A=&RandomRangeFloat(Minimum=0, Maximum=0.03)
        B=&RandomRangeFloat(Minimum=0.97, Maximum=1)
        Alpha=&SelectFloatFromArray(
         Direct Array Index=&ModuloInt(A=$Engine.Emitter.TotalSpawnedParticles, B=2)
         Float Selection Array=DataInterface() {propertyValues {FloatData [0,1]}}
         Random Seed=0
        )
       )
      )
     )
     Front End=&LerpPosition(
      A=<80 20 75>
      B=<-80 20 75>
      Alpha=&Lerp_Float(
       A=&RandomRangeFloat(Minimum=0, Maximum=0.03)
       B=&RandomRangeFloat(Minimum=0.97, Maximum=1)
       Alpha=&SelectFloatFromArray(
        Direct Array Index=&ModuloInt(A=$Engine.Emitter.TotalSpawnedParticles, B=2)
        Float Selection Array=DataInterface() {propertyValues {FloatData [0,1]}}
        Random Seed=0
       )
      )
     )
    )
    Offset Mode=::NewEnumerator0
    Override Seed=false
    Shape Primitive=::NewEnumerator2
    Surface Only=false
    UseOwnerScale=true
   )
   &InitialMeshOrientation(
    Mesh Orientation Mode=::NewEnumerator2
    Rotation=&RandomRangeVector(Minimum=<0 0 0>, Maximum=<0 0 1>)
    Use Rotation Vector=true
   )
   &CameraOffset(Camera Offset Amount=&MaskFloat(Mask Source=0, Mask Parameter=?, Mask=3, Masked Float=10, Passthrough Float=10))
   &AddVelocity(Velocity=&FN_FrontEndVector3Modifier(Gameplay=<0 0 0>, Front End=<0 -25 0>))
  ]
  particleUpdateScript [
   &ParticleState()
   &Timeline(
    Play=&SetBoolByFloatComparison(A=&VectorLength(Vector=$User.OwnerLocalVelocity), B=150)
    Play Rate=&RandomRangeFloat(Minimum=0.25, Maximum=5)
    Rewind Rate:NiagaraFloat=&RandomRangeFloat(Minimum=0.25, Maximum=5)
   )
   &ScaleMeshSize(Scale Factor=&VectorFromCurve(VectorCurve=DataInterface() {propertyValues {
       XCurve {
        defaultValue=0
        keys [
         {arriveTangent=4.5415, interpMode=RCIM_Cubic, leaveTangent=4.5415, tangentMode=RCTM_User, time=0.05, value=0}
         {arriveTangentWeight=0.0667, interpMode=RCIM_Cubic, leaveTangentWeight=0.2436, tangentMode=RCTM_Break, time=0.5, value=1}
         {arriveTangent=-4.8703, arriveTangentWeight=0.0667, interpMode=RCIM_Cubic, leaveTangent=-4.8703, tangentMode=RCTM_User, time=0.96, value=0}
        ]
       }
       YCurve {
        defaultValue=0
        keys [
         {arriveTangent=4.5415, interpMode=RCIM_Cubic, leaveTangent=4.5415, tangentMode=RCTM_User, time=0.05, value=0}
         {arriveTangentWeight=0.0667, interpMode=RCIM_Cubic, leaveTangentWeight=0.2436, tangentMode=RCTM_Break, time=0.5, value=1}
         {arriveTangent=-4.8703, arriveTangentWeight=0.0667, interpMode=RCIM_Cubic, leaveTangent=-4.8703, tangentMode=RCTM_User, time=0.96, value=0}
        ]
       }
       ZCurve {
        defaultValue=0
        keys [
         {arriveTangent=4.5415, interpMode=RCIM_Cubic, leaveTangent=4.5415, tangentMode=RCTM_User, time=0.05, value=0}
         {arriveTangentWeight=0.0667, interpMode=RCIM_Cubic, leaveTangentWeight=0.1722, tangentMode=RCTM_Break, time=0.5, value=1}
         {arriveTangent=-4.8703, arriveTangentWeight=0.0667, interpMode=RCIM_Cubic, leaveTangent=-4.8703, tangentMode=RCTM_User, time=0.96, value=0}
        ]
       }
      }}))
   /SetVariables(
    Particles.Update Glow Color:LinearColor=&LerpLinearColors(
     StartColor=&Multiply_LinearColorByFloat(
      Linear Color=$Particles.Initial.Color
      Float=&FNTimeOfDayScalar(Night Scale=0.75, Morning Scale=0.75, Day Scale=0.87, Evening Scale=0.75)
     )
     EndColor=&Multiply_LinearColorByFloat(
      Linear Color=<rgba 0.0676 0.39 1 0.33>
      Float=&FNTimeOfDayScalar(Night Scale=0.75, Morning Scale=0.75, Day Scale=0.87, Evening Scale=0.75)
     )
     LerpFactor=&Waveform(
      "Waveform [1]"=::NewEnumerator0
      "[1] Frequency"=&RandomRangeFloat(Minimum=0.15, Maximum=0.7)
      "[1] Amplitude Min/Max"=<0 1>
     )
    )
    Particles.Update Glow Sprite Size:Vector2f=&Multiply_Vector2D(
     A=$Particles.Initial.SpriteSize
     B=&Multiply_Vector2D_ByFloat(
      Float=&Waveform(
       "Waveform [1]"=::NewEnumerator0
       "[1] Frequency"=&RandomRangeFloat(Minimum=0.3, Maximum=0.6)
       "[1] Amplitude Min/Max"=<0 0.18>
       "[1] Bias"=0.8
      )
      Vector2D=&Vector2DFromCurve(
       Vector2Curve=DataInterface() {propertyValues {
         XCurve {
          defaultValue=1
          keys [
           {interpMode=RCIM_Cubic, tangentMode=RCTM_User, time=0, value=0}
           {arriveTangentWeight=0.1167, interpMode=RCIM_Cubic, tangentMode=RCTM_User, time=0.35, value=1}
           {arriveTangentWeight=0.1, interpMode=RCIM_Cubic, tangentMode=RCTM_User, time=0.65, value=1}
           {arriveTangentWeight=0.1167, interpMode=RCIM_Cubic, tangentMode=RCTM_User, time=1, value=0}
          ]
         }
         YCurve {
          defaultValue=1
          keys [
           {interpMode=RCIM_Cubic, tangentMode=RCTM_User, time=0, value=0}
           {arriveTangentWeight=0.1167, interpMode=RCIM_Cubic, tangentMode=RCTM_User, time=0.35, value=1}
           {arriveTangentWeight=0.1, interpMode=RCIM_Cubic, tangentMode=RCTM_User, time=0.65, value=1}
           {arriveTangentWeight=0.1167, interpMode=RCIM_Cubic, tangentMode=RCTM_User, time=1, value=0}
          ]
         }
        }}
       ScaleCurve=<1 1>
      )
     )
    )
   ) {isSetParameters=true}
   /SetVariables(
    Particles.Update Spawn Glow Color:LinearColor=&Multiply_LinearColorByFloat(
     Linear Color=&Multiply_LinearColorByFloat(
      Linear Color=&Multiply_LinearColorByFloat(
       Linear Color=&Multiply_LinearColorByFloat(
        Linear Color=<rgba 0.0016 0.4055 1 0.1>
        Float=&FNTimeOfDayScalar(Night Scale=0.7, Morning Scale=0.82, Evening Scale=0.82)
       )
       Float=&FloatFromCurve(FloatCurve=DataInterface() {propertyValues {Curve {keys [
            {arriveTangent=2.9184, interpMode=RCIM_Cubic, leaveTangent=2.9184, tangentMode=RCTM_User, time=0, value=0}
            {arriveTangent=15.7671, interpMode=RCIM_Cubic, leaveTangent=15.7671, tangentMode=RCTM_User, time=0.05, value=0.3}
            {arriveTangent=1.5023, interpMode=RCIM_Cubic, leaveTangent=1.5023, tangentMode=RCTM_User, time=0.15, value=0.93}
            {arriveTangent=-6.4306, interpMode=RCIM_Cubic, leaveTangent=-6.4306, tangentMode=RCTM_User, time=0.28, value=0.425}
            {arriveTangent=-0.0209, interpMode=RCIM_Cubic, leaveTangent=-0.0209, tangentMode=RCTM_User, time=0.4, value=0}
            {arriveTangent=0.9375, interpMode=RCIM_Cubic, leaveTangent=0.9375, time=0.68, value=0}
            {arriveTangent=15.8107, interpMode=RCIM_Cubic, leaveTangent=15.8107, tangentMode=RCTM_User, time=0.72, value=0.3}
            {arriveTangent=0.4413, interpMode=RCIM_Cubic, leaveTangent=0.4413, tangentMode=RCTM_User, time=0.8, value=0.85}
            {arriveTangent=-6.515, interpMode=RCIM_Cubic, leaveTangent=-6.515, tangentMode=RCTM_User, time=1, value=0}
           ]}}})
      )
      Float=&FNTimeOfDayScalar(Night Scale=0.7, Morning Scale=0.82, Evening Scale=0.82)
     )
     Float=&FloatFromCurve(FloatCurve=DataInterface() {propertyValues {Curve {keys [
          {arriveTangent=2.9184, interpMode=RCIM_Cubic, leaveTangent=2.9184, tangentMode=RCTM_User, time=0, value=0}
          {arriveTangent=15.7671, interpMode=RCIM_Cubic, leaveTangent=15.7671, tangentMode=RCTM_User, time=0.05, value=0.3}
          {arriveTangent=1.5023, interpMode=RCIM_Cubic, leaveTangent=1.5023, tangentMode=RCTM_User, time=0.15, value=0.93}
          {arriveTangent=-6.4306, interpMode=RCIM_Cubic, leaveTangent=-6.4306, tangentMode=RCTM_User, time=0.28, value=0.425}
          {arriveTangent=-0.0209, interpMode=RCIM_Cubic, leaveTangent=-0.0209, tangentMode=RCTM_User, time=0.4, value=0}
          {arriveTangent=0.9375, interpMode=RCIM_Cubic, leaveTangent=0.9375, time=0.68, value=0}
          {arriveTangent=15.8107, interpMode=RCIM_Cubic, leaveTangent=15.8107, tangentMode=RCTM_User, time=0.72, value=0.3}
          {arriveTangent=0.4413, interpMode=RCIM_Cubic, leaveTangent=0.4413, tangentMode=RCTM_User, time=0.8, value=0.85}
          {arriveTangent=-6.515, interpMode=RCIM_Cubic, leaveTangent=-6.515, tangentMode=RCTM_User, time=1, value=0}
         ]}}})
    )
    Particles.Update Spawn Glow Sprite Size:Vector2f=&Multiply_Vector2D(
     A=$Particles.Initial.SpriteSize
     B=&Multiply_Vector2D_ByFloat(
      Float=&FloatFromCurve(
       FloatCurve=DataInterface() {propertyValues {Curve {
          defaultValue=1
          keys [
           {arriveTangent=7.1914, interpMode=RCIM_Cubic, leaveTangent=7.1914, tangentMode=RCTM_User, time=0, value=0}
           {arriveTangent=8.0734, arriveTangentWeight=0.03, interpMode=RCIM_Cubic, leaveTangent=8.0734, tangentMode=RCTM_User, time=0.02, value=0.25}
           {arriveTangent=-0.3101, interpMode=RCIM_Cubic, leaveTangent=-0.3101, tangentMode=RCTM_User, time=0.1, value=0.45}
           {arriveTangent=-1.6156, arriveTangentWeight=0.0333, interpMode=RCIM_Cubic, leaveTangent=-1.6156, tangentMode=RCTM_User, time=0.2, value=0.2}
           {arriveTangentWeight=0.034, interpMode=RCIM_Cubic, leaveTangentWeight=0.136, tangentMode=RCTM_User, time=0.4, value=0}
           {arriveTangentWeight=0.1477, interpMode=RCIM_Cubic, leaveTangentWeight=0.0074, tangentMode=RCTM_User, time=0.68, value=0}
           {arriveTangent=10.5287, interpMode=RCIM_Cubic, leaveTangent=10.5287, tangentMode=RCTM_User, time=0.72, value=0.2}
           {
            arriveTangent=0.5653
            arriveTangentWeight=0.0333
            interpMode=RCIM_Cubic
            leaveTangent=0.5653
            leaveTangentWeight=0.0333
            tangentMode=RCTM_User
            time=0.8
            value=0.5
           }
           {arriveTangent=-3.5336, tangentMode=RCTM_User, time=1, value=0.3}
          ]
         }}}
       Scale Curve=1.2
      )
      Vector2D=&RandomRangeVector2D(Minimum=<1.2 1.2>, Maximum=<1.25 1.45>)
     )
    )
   ) {isSetParameters=true}
   /SetVariables(
    Particles.Update Despawn Flare Color:LinearColor=<rgba 0.251 0.8073 1 0.3>
    Particles.Update Despawn Flare Rotation:NiagaraFloat=&RandomRangeFloat(Minimum=-360, Maximum=360)
    Particles.Update Despawn Flare Sprite Size:Vector2f=&Multiply_Vector2D_ByFloat(
     Float=&FloatFromCurve(FloatCurve=DataInterface() {propertyValues {Curve {keys [
          {interpMode=RCIM_Cubic, tangentMode=RCTM_User, time=0.6, value=0}
          {arriveTangent=0.2424, interpMode=RCIM_Cubic, leaveTangent=0.2424, time=0.8, value=0.08}
          {arriveTangent=-1.0457, interpMode=RCIM_Cubic, leaveTangent=-1.0457, tangentMode=RCTM_User, time=0.93, value=0.08}
          {arriveTangent=1.2006, interpMode=RCIM_Cubic, leaveTangent=1.2006, tangentMode=RCTM_User, time=0.98, value=0.4}
         ]}}})
     Vector2D=$Particles.Initial.SpriteSize
    )
   ) {isSetParameters=true}
   /SetVariables(
    Particles.Bubble Color:LinearColor=&Multiply_LinearColorByFloat(
     Linear Color=<rgba 1 1 1 0.05>
     Float=&FloatFromCurve(FloatCurve=DataInterface() {propertyValues {Curve {
         defaultValue=0
         keys [
          <at 0 0>
          <at 0.75 0>
          {arriveTangent=-0.0858, arriveTangentWeight=0.2333, interpMode=RCIM_Cubic, leaveTangent=-0.0858, tangentMode=RCTM_User, time=0.8, value=0.005}
          {arriveTangent=8.9897, interpMode=RCIM_Cubic, leaveTangent=8.9897, tangentMode=RCTM_User, time=0.88, value=0.25}
          {
           arriveTangent=0.3863
           arriveTangentWeight=0.0833
           interpMode=RCIM_Cubic
           leaveTangent=0.3863
           leaveTangentWeight=0.0333
           tangentMode=RCTM_User
           time=0.9
           value=0.65
          }
          {arriveTangent=-12.4305, interpMode=RCIM_Cubic, leaveTangent=-12.4305, tangentMode=RCTM_User, time=1, value=0}
         ]
        }}})
    )
    Particles.Bubble Dynamic Param:Vector4f=&MakeVector4(X=0, Y=0, Z=0, W=0)
    Particles.Bubble Sprite Size:Vector2f=&Multiply_Vector2D_ByFloat(
     Float=&FloatFromCurve(FloatCurve=DataInterface() {propertyValues {Curve {keys [
          <at 0 0>
          {arriveTangent=-0.0108, interpMode=RCIM_Cubic, leaveTangent=-0.0108, tangentMode=RCTM_User, time=0.55, value=0}
          {arriveTangent=0.3115, interpMode=RCIM_Cubic, leaveTangent=0.3115, tangentMode=RCTM_User, time=0.8, value=0.04}
          {arriveTangent=-0.295, interpMode=RCIM_Cubic, leaveTangent=12.8805, tangentMode=RCTM_Break, time=0.97, value=0.08}
          {interpMode=RCIM_Cubic, leaveTangent=-0.1067, time=1, value=0.25}
         ]}}})
     Vector2D=$Particles.Initial.SpriteSize
    )
   ) {isSetParameters=true}
   &BlendSpace("Position Blend (World <-> Local)":NiagaraFloat=&MakeCustomFloatFromBool(
     Bool=&SetBoolByFloatComparison(A=&VectorLength(Vector=$Engine.Owner.Velocity), B=301)
     True Float=1
     False Float=&FloatFromCurve(FloatCurve=DataInterface() {propertyValues {Curve {
         defaultValue=0
         keys [{interpMode=RCIM_Cubic, tangentMode=RCTM_User, time=0, value=1}, {interpMode=RCIM_Cubic, tangentMode=RCTM_User, time=0.55, value=0.55}]
        }}})
     Translation Contribution:NiagaraFloat=1
     Rotation Contribution:NiagaraFloat=1
    )) {module=BlendSpace001}
   &GravityForce(
    Coordinate Space=::Local
    Gravity=&MakeCustomVectorFromBool(
     Bool=&SetBoolByFloatComparison(A=&VectorLength(Vector=$User.OwnerLocalVelocity), B=50)
     True Vector=<0 4 1>
     False Vector=<0 0 3>
    )
   )
   &SolveForcesAndVelocity(Clamp Velocity=false, Limit Acceleration=false, Speed Limit=50)
   /SetVariables(Particles.Position=&AddVectorToPosition(
     Position=$Particles.Position
     Vector=&MakeCustomVectorFromBool(
      Bool=&SetBoolByFloatComparison(A=&VectorLength(Vector=$User.OwnerLocalVelocity), B=50)
      True Vector=&MakeVector(
       X=0
       Y=&MakeCustomFloatFromBool(
        Bool=&SetBoolByFloatComparison(A=&VectorLength(Vector=$Engine.Owner.Velocity), B=310)
        True Float=&FloatFromCurve(FloatCurve=DataInterface() {propertyValues {Curve {keys [<at 0 1>,<at 1 0>]}}}, Scale Curve=-4)
        False Float=&FloatFromCurve(FloatCurve=DataInterface() {propertyValues {Curve {keys [<at 0 1>,<at 1 0>]}}}, Scale Curve=0.3)
       )
       Z=&FloatFromCurve(
        FloatCurve=DataInterface() {propertyValues {Curve {defaultValue=1, keys [{tangentMode=RCTM_User, time=0, value=0}, {tangentMode=RCTM_User, time=1, value=1}]}}}
        Scale Curve=0.5
       )
      )
      False Vector=&MakeVector(
       X=0
       Y=&FN_FrontEndFloatModifier(
        Gameplay=0
        Front End=&FloatFromCurve(FloatCurve=DataInterface() {propertyValues {Curve {keys [<at 0 1>,<at 1 0>]}}}, Scale Curve=0.3)
       )
       Z=&FN_FrontEndFloatModifier(
        Gameplay=&FloatFromCurve(
         FloatCurve=DataInterface() {propertyValues {Curve {defaultValue=1, keys [{tangentMode=RCTM_User, time=0, value=0}, {tangentMode=RCTM_User, time=1, value=1}]}}}
         Scale Curve=0.5
        )
        Front End=&FloatFromCurve(
         FloatCurve=DataInterface() {propertyValues {Curve {defaultValue=1, keys [{tangentMode=RCTM_User, time=0, value=0}, {tangentMode=RCTM_User, time=1, value=1}]}}}
         Scale Curve=0.2
        )
       )
      )
     )
    )) {isSetParameters=true}
   &UpdateMeshOrientation(
    Facing Interpolation=true
    Orientation Method=::NewEnumerator1
    Override Quat=false
    TransformFacing=false
    TransformUp=false
    Up Direction:Vector3f=&Lerp_Vector(
     A=&FN_FrontEndVector3Modifier(Gameplay=<0 0 1>, Front End=<0 -0.8 1>)
     B=<0 -0.8 1>
     Alpha=$Particles.Timeline.NormalizedTime
    )
   )
  ]
  rendererClasses [&NiagaraMeshRendererProperties, &NiagaraSpriteRendererProperties]
  renderers [
   {rendererClass=&NiagaraMeshRendererProperties, rendererIndex=0}
   {rendererClass=&NiagaraSpriteRendererProperties, rendererIndex=1}
   {rendererClass=&NiagaraSpriteRendererProperties, rendererIndex=2}
   {rendererClass=&NiagaraSpriteRendererProperties, rendererIndex=3}
   {rendererClass=&NiagaraSpriteRendererProperties, rendererIndex=4}
  ]
 }
 Light {
  emitterUpdateScript [
   &EmitterState(Life Cycle Mode=::NewEnumerator1, Loop Duration=1, UseLoopDelay=false)
   &SpawnBurst_Instantaneous(Use Spawn Probability=false, UseLoopCountLimit=false)
  ]
  particleSpawnScript [&InitializeParticle(Color Mode=::NewEnumerator0, UsePositionOffset=false, Write Material Random=true)]
  particleUpdateScript [
   &ParticleState()
   &FN_Light_Attributes(
    Color=<rgba 0 0.5214 1 1>
    Create Color Attribute=true
    Create Exponent Attribute=true
    Create Position Attribute=true
    Create Radius Attribute=true
    Enable Particle Light Enable=true
    Exponent=2
    Light Enabled=true
    Position=<0 0 200>
    Radius=/Niagara/DynamicInputs/TypeConversions/Vector2DLength(Vector:Vector2f=<85 275>)
    Scale Color=10
   )
   &FN_Gen5_Light(Color:LinearColor=<rgba 0 0.5214 1 1>, Intensity=15000)
  ]
  rendererClasses [&NiagaraLightRendererProperties]
  renderers [{rendererClass=&NiagaraLightRendererProperties, rendererIndex=0}]
 }
 PropellerBubbles {
  emitterUpdateScript [
   &EmitterState(Life Cycle Mode=::NewEnumerator1, Loop Duration=1, UseLoopDelay=false)
   &SpawnRate(
    SpawnRate=&MakeCustomFloatFromBool(
     Bool=&SetBoolByFloatComparison(A=&VectorLength(Vector=$User.OwnerLocalVelocity), B=50)
     True Float=&RandomRangeFloat(Minimum=20, Maximum=27)
     False Float=10
    )
    Use Spawn Probability=false
   )
   &self:InterpolateSpeedThresholdEmitter_04(
    Define Initial Value=false
    Target Float Value:NiagaraFloat=&MakeCustomFloatFromBool(
     Bool=&SetBoolByFloatComparison(
      A=&Add_Float(
       A=&Add_Float(
        A=&Abs_Float(Float:NiagaraFloat=&MakeFloatFromVector(Vector=$Engine.Owner.Velocity, Channel=::NewEnumerator0))
        B=&Abs_Float(Float:NiagaraFloat=&MakeFloatFromVector(Vector=$Engine.Owner.Velocity, Channel=::NewEnumerator2))
       )
       B=&Abs_Float(Float:NiagaraFloat=&MakeFloatFromVector(Vector=$Engine.Owner.Velocity, Channel=::NewEnumerator1))
      )
      B=650
     )
     True Float=0.88
     False Float=0
    )
   ) {module=InterpolateSpeedThresholdEmitter_04}
  ]
  particleSpawnScript [
   &InitializeParticle(
    Color=<rgba 0 0.3112 0.674 1>
    Lifetime Max=0.7
    Lifetime Min=0.4
    Lifetime Mode=::NewEnumerator1
    Sprite Rotation Angle=&RandomRangeFloat(Minimum=-0.2, Maximum=-0.1)
    Sprite Rotation Mode=::NewEnumerator1
    Sprite Size Mode=::NewEnumerator3
    Uniform Sprite Size=&MakeCustomFloatFromBool(
     Bool=&SetBoolByFloatComparison(A=&VectorLength(Vector=$User.OwnerLocalVelocity), B=50)
     True Float=&RandomRangeFloat(Minimum=0.8, Maximum=5)
     False Float=&FN_FrontEndFloatModifier(
      Gameplay=&RandomRangeFloat(Minimum=0.5, Maximum=2)
      Front End=&RandomRangeFloat(Minimum=0.8, Maximum=5)
     )
    )
    UsePositionOffset=false
    Write Material Random=true
   )
   &AddVelocity(Velocity=&MakeCustomVectorFromBool(
     Bool=&SetBoolByFloatComparison(A=&VectorLength(Vector=$User.OwnerLocalVelocity), B=50)
     True Vector=<0 -20 1>
     False Vector=&FN_FrontEndVector3Modifier(Gameplay=&RandomRangeVector(Minimum=<0 -120 0>, Maximum=<1 -150 50>), Front End=<0 -165 1>)
    )) {module=AddVelocity001}
   &SkeletalMeshLocation(
    Bone Sampling Mode=::NewEnumerator4
    MaskSpawnGroup=false
    Position Sampling=::NewEnumerator1
    Skeletal Mesh=$System.Propeller Bubbles Spawn
    Source Movement Scale:Vector3f=<0.5 0.5 0.5>
   ) {module=SkeletalMeshLocation001}
   &ShapeLocation(MaskSpawnGroup=false, Non Uniform Scale=<1 0.3 0.5>, Override Seed=false, Sphere Radius=8, UseOwnerScale=true)
  ]
  particleUpdateScript [
   &CurlNoiseForce(
    Noise Strength=&FloatFromCurve(FloatCurve=DataInterface() {propertyValues {Curve {keys [
         {interpMode=RCIM_Cubic, tangentMode=RCTM_User, time=0, value=300}
         {arriveTangent=772.9872, interpMode=RCIM_Cubic, leaveTangent=772.9872, tangentMode=RCTM_User, time=1, value=550}
        ]}}})
    Pan Noise=false
   )
   &GravityForce(Gravity=&RandomRangeVector(
     Minimum=&VectorFromCurve(VectorCurve=DataInterface() {propertyValues {
        XCurve {keys [<at 0 0>,<at 1 0>]}
        YCurve {keys [<at 0 0>,<at 1 0>]}
        ZCurve {keys [
          {interpMode=RCIM_Cubic, tangentMode=RCTM_User, time=0, value=10}
          {arriveTangent=88.4209, interpMode=RCIM_Cubic, leaveTangent=88.4209, tangentMode=RCTM_User, time=0.75, value=35}
         ]}
       }})
     Maximum=&VectorFromCurve(VectorCurve=DataInterface() {propertyValues {
        XCurve {keys [<at 0 0>,<at 1 0>]}
        YCurve {keys [<at 0 0>,<at 1 0>]}
        ZCurve {keys [
          {interpMode=RCIM_Cubic, tangentMode=RCTM_User, time=0, value=10}
          {arriveTangent=133.5137, interpMode=RCIM_Cubic, leaveTangent=133.5137, tangentMode=RCTM_User, time=0.75, value=70}
         ]}
       }})
    ))
   &DynamicMaterialParameters(
    Index 0 Param 1=&FloatFromCurve(FloatCurve=DataInterface() {propertyValues {Curve {keys [
         {arriveTangent=0.2342, interpMode=RCIM_Cubic, leaveTangent=0.2342, tangentMode=RCTM_User, time=0, value=0.2}
         {arriveTangent=0.2651, interpMode=RCIM_Cubic, leaveTangent=0.2651, tangentMode=RCTM_User, time=0.9, value=0.25}
         {arriveTangent=3.2005, interpMode=RCIM_Cubic, leaveTangent=3.2005, tangentMode=RCTM_User, time=1, value=0.75}
        ]}}})
    Module.Param0WriteEnabled=true
    Module.Param1WriteEnabled=true
    Module.Param2WriteEnabled=true
    Module.Param3WriteEnabled=true
   )
   &Drag(Drag=&RandomRangeFloat(Minimum=1.5, Maximum=2.5), Use Linear Drag=true, Use Rotational Drag=false)
   &InheritSourceMovement(
    Applied Position Scale=&VectorFromFloat(Value=$Emitter.InterpolateSpeedThresholdEmitter_04.InterpolateOverTime_Float.Moving Average)
    ApplyForce=false
    ApplyPosition=true
    ApplyPositionFalloff=false
   )
   &ScaleColor(
    Scale Alpha=&FloatFromCurve(FloatCurve=DataInterface() {propertyValues {Curve {keys [
         {interpMode=RCIM_Cubic, leaveTangentWeight=0.1216, tangentMode=RCTM_User, time=0, value=0.95}
         {arriveTangent=-2.4777, interpMode=RCIM_Cubic, leaveTangent=-2.4777, tangentMode=RCTM_User, time=0.26, value=0.75}
         {arriveTangent=-2.274, interpMode=RCIM_Cubic, leaveTangent=-2.274, tangentMode=RCTM_User, time=0.4, value=0.35}
         {arriveTangent=-0.4872, interpMode=RCIM_Cubic, leaveTangent=-0.4872, tangentMode=RCTM_User, time=0.6, value=0.08}
         {arriveTangentWeight=0.1871, interpMode=RCIM_Cubic, tangentMode=RCTM_User, time=1, value=0}
        ]}}})
    ScaleA=true
    ScaleRGB=true
   )
   &ParticleState()
   &ScaleSpriteSize(Scale Factor=&Vector2DFromFloat(Value=&FloatFromCurve(
      FloatCurve=DataInterface() {propertyValues {Curve {keys [
          {arriveTangent=4.1785, interpMode=RCIM_Cubic, leaveTangent=4.1785, tangentMode=RCTM_User, time=0, value=0}
          {arriveTangent=0.9995, interpMode=RCIM_Cubic, leaveTangent=0.9995, tangentMode=RCTM_User, time=0.1, value=0.35}
          <at 0.9 0.75>
          {arriveTangent=2.5, interpMode=RCIM_Cubic, leaveTangent=2.5, time=0.95, value=0.57}
          {arriveTangent=4.4385, interpMode=RCIM_Cubic, leaveTangent=4.4385, tangentMode=RCTM_User, time=1, value=1}
         ]}}}
      Scale Curve=4
     )))
   &SpriteRotationRate(Rotation Rate=&RandomRangeFloat(Minimum=-40, Maximum=40))
   &AccelerationForce(Acceleration=&MakeCustomVectorFromBool(
     Bool=&SetBoolByFloatComparison(
      A=&Add_Float(
       A=&Add_Float(
        A=&Abs_Float(Float:NiagaraFloat=&MakeFloatFromVector(Vector=$Engine.Owner.Velocity, Channel=::NewEnumerator0))
        B=&Abs_Float(Float:NiagaraFloat=&MakeFloatFromVector(Vector=$Engine.Owner.Velocity, Channel=::NewEnumerator2))
       )
       B=&Abs_Float(Float:NiagaraFloat=&MakeFloatFromVector(Vector=$Engine.Owner.Velocity, Channel=::NewEnumerator1))
      )
      B=50
     )
     True Vector=<0 0 12>
     False Vector=&VectorFromCurve(VectorCurve=DataInterface() {propertyValues {
        XCurve {keys [<at 0 0>,<at 1 0>]}
        YCurve {keys [{arriveTangent=-191.6948, interpMode=RCIM_Cubic, leaveTangent=-191.6948, tangentMode=RCTM_User, time=0.2, value=0}, <at 1 -50 RCIM_Cubic>]}
        ZCurve {keys [<at 0 0>,<at 1 0>]}
       }})
    ))
   &SolveForcesAndVelocity(Clamp Velocity=false, Limit Acceleration=false)
  ]
  rendererClasses [&NiagaraSpriteRendererProperties]
  renderers [{rendererClass=&NiagaraSpriteRendererProperties}]
 }
}
summary {
 systemName=NS_Emote_PartyJelly
 emitters [
  {emitterName=PropellerBubbles, rendererClasses [&NiagaraSpriteRendererProperties]}
  {emitterName=BubbleBurst, rendererClasses [&NiagaraSpriteRendererProperties]}
  {emitterName=Light, rendererClasses [&NiagaraLightRendererProperties]}
  {emitterName=AmbientBubbles001, rendererClasses [&NiagaraSpriteRendererProperties]}
  {emitterName=Bubbles_Despawn, rendererClasses [&NiagaraSpriteRendererProperties]}
  {emitterName=Bubbles_Spawn, rendererClasses [&NiagaraSpriteRendererProperties]}
  {emitterName=Jellyfish, rendererClasses [&NiagaraMeshRendererProperties, &NiagaraSpriteRendererProperties]}
  {emitterName=Droplets, rendererClasses [&NiagaraSpriteRendererProperties]}
 ]
 userVariables [
  User.bIsFrontEnd {type=NiagaraBool}
  User.bIsFrontEndPreview {type=NiagaraBool}
  User.bIsInWater {type=NiagaraBool}
  User.OwnerLocalVelocity {type=Vector3f, description="", defaultValue {struct=Vector3f, value=<0 0 0>}}
 ]
}
