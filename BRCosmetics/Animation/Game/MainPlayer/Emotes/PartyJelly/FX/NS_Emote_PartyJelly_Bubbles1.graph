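# Niagara effect for the PartyJelly emote: blue bubbles around the character, with drifting jellyfish.
# Bubbles spawn on User.Skeletal Mesh, rise with curl noise, and follow owner movement through an interpolated scale (target 0.88 regardless of speed).
# Jellyfish001 alternates mesh-plus-sprite jellyfish between two side offsets, bobbing, rotating and scaling in and out.
# User.OwnerLocalVelocity raises both spawn rates above separate speed thresholds; Jellyfish and FN_Debug emitters are disabled.
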NiagaraSystem /BRCosmetics/Animation/Game/MainPlayer/Emotes/PartyJelly/FX/NS_Emote_PartyJelly_Bubbles1
dependencies {
 usedDataInterfaces [
  &NiagaraDataInterfaceArrayFloat
  &NiagaraDataInterfaceCurve
  &NiagaraDataInterfaceVectorCurve
  &NiagaraDataInterfaceStaticMesh
 ]
 usedDynamicInputs []
 usedModules [&SystemState]
}
emitters {
 Bubbles {
  emitterUpdateScript [
   &EmitterState(Life Cycle Mode=::NewEnumerator1, Loop Duration=1, UseLoopDelay=false)
   &SpawnRate(
    SpawnRate=&MakeCustomFloatFromBool(
     Bool=&SetBoolByFloatComparison(A=&VectorLength(Vector=$User.OwnerLocalVelocity), B=50)
     True Float=&RandomRangeFloat(Minimum=45, Maximum=60)
     False Float=15
    )
    Use Spawn Probability=false
   )
   &self:InterpolateSpeedThresholdEmitter_04(
    Define Initial Value=false
    Rate Of Change=4
    Target Float Value:NiagaraFloat=&MakeCustomFloatFromBool(
     Bool=&SetBoolByFloatComparison(
      A=&Add_Float(
       A=&Add_Float(
        A=&Abs_Float(Float:NiagaraFloat=&MakeFloatFromVector(Vector=$Engine.Owner.Velocity, Channel=::NewEnumerator0))
        B=&Abs_Float(Float:NiagaraFloat=&MakeFloatFromVector(Vector=$Engine.Owner.Velocity, Channel=::NewEnumerator2))
       )
       B=&Abs_Float(Float:NiagaraFloat=&MakeFloatFromVector(Vector=$Engine.Owner.Velocity, Channel=::NewEnumerator1))
      )
      B=650
     )
     True Float=0.88
     False Float=0.88
    )
   ) {module=InterpolateSpeedThresholdEmitter_04}
  ]
  particleSpawnScript [
   &InitializeParticle(
    Color=<rgba 0 0.3112 0.674 1>
    Lifetime Max=1
    Lifetime Min=0.5
    Lifetime Mode=::NewEnumerator1
    Sprite Rotation Angle=&RandomRangeFloat(Minimum=-0.2, Maximum=-0.1)
    Sprite Rotation Mode=::NewEnumerator1
    Sprite Size Mode=::NewEnumerator3
    Uniform Sprite Size=&SelectFloatFromArray(
     Array Sampling Mode=::NewEnumerator0
     Direct Array Index=?
     Float Selection Array=DataInterface() {propertyValues {FloatData [1.3,1.5,2,2.5,4,4.5,5]}}
     Random Seed=0
    )
    UsePositionOffset=false
    Write Material Random=true
   )
   &AddVelocity(Velocity=&RandomRangeVector(Minimum=<0 -120 0>, Maximum=<1 -170 50>))
   &SkeletalMeshLocation(Bone Sampling Mode=::NewEnumerator4, MaskSpawnGroup=false, Position Sampling=::NewEnumerator1, Skeletal Mesh=$User.Skeletal Mesh)
   &ShapeLocation(MaskSpawnGroup=false, Non Uniform Scale=<1.32 1 1>, Override Seed=false, Sphere Radius=6, UseOwnerScale=true)
  ]
  particleUpdateScript [
   &CurlNoiseForce(Noise Strength=300, Pan Noise=false)
   &GravityForce(Gravity=<0 0 50>)
   &DynamicMaterialParameters(
    Index 0 Param 1=&FloatFromCurve(FloatCurve=DataInterface() {propertyValues {Curve {keys [
         {arriveTangent=0.2342, interpMode=RCIM_Cubic, leaveTangent=0.2342, tangentMode=RCTM_User, time=0, value=0.2}
         {arriveTangent=0.2651, interpMode=RCIM_Cubic, leaveTangent=0.2651, tangentMode=RCTM_User, time=0.9, value=0.25}
         {arriveTangent=3.2005, interpMode=RCIM_Cubic, leaveTangent=3.2005, tangentMode=RCTM_User, time=1, value=0.75}
        ]}}})
    Module.Param0WriteEnabled=true
    Module.Param1WriteEnabled=true
    Module.Param2WriteEnabled=true
    Module.Param3WriteEnabled=true
   )
   &Drag(Drag=&RandomRangeFloat(Minimum=1.5, Maximum=2.5), Use Linear Drag=true, Use Rotational Drag=false)
   &InheritSourceMovement(
    Applied Position Scale=&VectorFromFloat(Value=$Emitter.InterpolateSpeedThresholdEmitter_04.InterpolateOverTime_Float.Moving Average)
    ApplyForce=false
    ApplyPosition=true
    ApplyPositionFalloff=false
   )
   &AccelerationForce(Acceleration=&RandomRangeVector(Minimum=<-100 -100 0>, Maximum=<100 100 0>))
   &ScaleColor(
    Scale Alpha=&FloatFromCurve(FloatCurve=DataInterface() {propertyValues {Curve {keys [
         {interpMode=RCIM_Cubic, tangentMode=RCTM_User, time=0, value=1}
         {arriveTangent=-0.6963, interpMode=RCIM_Cubic, leaveTangent=-0.6963, tangentMode=RCTM_User, time=0.37, value=0.95}
         {arriveTangent=-1.5079, interpMode=RCIM_Cubic, leaveTangent=-1.5079, time=0.7, value=0.2}
         {interpMode=RCIM_Cubic, tangentMode=RCTM_User, time=1, value=0}
        ]}}})
    ScaleA=true
    ScaleRGB=true
   )
   &ParticleState()
   &ScaleSpriteSize(Scale Factor=&Vector2DFromFloat(Value=&FloatFromCurve(
      FloatCurve=DataInterface() {propertyValues {Curve {keys [
          {arriveTangent=4.1785, interpMode=RCIM_Cubic, leaveTangent=4.1785, tangentMode=RCTM_User, time=0, value=0}
          {arriveTangent=0.9995, interpMode=RCIM_Cubic, leaveTangent=0.9995, tangentMode=RCTM_User, time=0.1, value=0.35}
          <at 0.9 0.75>
          {arriveTangent=2.5, interpMode=RCIM_Cubic, leaveTangent=2.5, time=0.95, value=0.57}
          {arriveTangent=4.4385, interpMode=RCIM_Cubic, leaveTangent=4.4385, tangentMode=RCTM_User, time=1, value=1}
         ]}}}
      Scale Curve=4
     )))
   &SpriteRotationRate(Rotation Rate=&RandomRangeFloat(Minimum=-40, Maximum=40))
   &SolveForcesAndVelocity(Clamp Velocity=false, Limit Acceleration=false)
  ]
  rendererClasses [&NiagaraSpriteRendererProperties]
  renderers [{rendererClass=&NiagaraSpriteRendererProperties}]
 }
 FN_Debug {
  bEnabled=false
  emitterUpdateScript [
   &EmitterState()
   &SpawnBurst_Instantaneous(Spawn Count=1, Use Spawn Probability=false)
   &self:InterpolateSpeedThresholdEmitter_04(
    Define Initial Value=false
    Rate Of Change=4
    Target Float Value:NiagaraFloat=&MakeCustomFloatFromBool(
     Bool=&SetBoolByFloatComparison(
      A=&Add_Float(
       A=&Add_Float(
        A=&Abs_Float(Float:NiagaraFloat=&MakeFloatFromVector(Vector=$Engine.Owner.Velocity, Channel=::NewEnumerator0))
        B=&Abs_Float(Float:NiagaraFloat=&MakeFloatFromVector(Vector=$Engine.Owner.Velocity, Channel=::NewEnumerator2))
       )
       B=&Abs_Float(Float:NiagaraFloat=&MakeFloatFromVector(Vector=$Engine.Owner.Velocity, Channel=::NewEnumerator1))
      )
      B=650
     )
     True Float=0.88
     False Float=0.88
    )
   ) {module=InterpolateSpeedThresholdEmitter_04}
  ]
  particleSpawnScript [/Niagara/Modules/Spawn/Initialization/InitializeParticle(
    Lifetime=5
    Module.Write Color=true
    Module.Write Lifetime=true
    Module.Write Mass=true
    Module.Write Material Random=true
    Module.Write Position=true
    Module.Write Scale=false
    Module.Write SpriteRotation=false
    Module.Write SpriteSize=true
    Position=&AddVectorToPosition(Position=/Niagara/DynamicInputs/Helpers/SimulationPosition(), Vector=<-100 0 100>)
    Sprite Size=<100 100>
   )]
  particleUpdateScript [
   &ParticleState()
   /SetVariables(Particles.Debug:NiagaraFloat=&Add_Float(
     A=&Add_Float(
      A=&Abs_Float(Float:NiagaraFloat=&MakeFloatFromVector(Vector=$Engine.Owner.Velocity, Channel=::NewEnumerator0))
      B=&Abs_Float(Float:NiagaraFloat=&MakeFloatFromVector(Vector=$Engine.Owner.Velocity, Channel=::NewEnumerator2))
     )
     B=&Abs_Float(Float:NiagaraFloat=&MakeFloatFromVector(Vector=$Engine.Owner.Velocity, Channel=::NewEnumerator1))
    )) {isSetParameters=true}
   &DynamicMaterialParameters(
    Index 0 Param 1=$Particles.Debug
    Module.Param0WriteEnabled=true
    Module.Param1WriteEnabled=true
    Module.Param2WriteEnabled=true
    Module.Param3WriteEnabled=true
   )
  ]
  rendererClasses [&NiagaraSpriteRendererProperties]
  renderers [{rendererClass=&NiagaraSpriteRendererProperties}]
 }
 Jellyfish {
  bEnabled=false
  emitterUpdateScript [
   &EmitterState(Life Cycle Mode=::NewEnumerator1, Loop Delay=0.1, Loop Duration=1, UseLoopDelay=false)
   &SpawnRate(
    SpawnRate=&MakeCustomFloatFromBool(
     Bool=&SetBoolByFloatComparison(A=&VectorLength(Vector=$User.OwnerLocalVelocity), B=50)
     True Float=&RandomRangeFloat(Minimum=8, Maximum=11)
     False Float=&RandomRangeFloat(Minimum=6, Maximum=7)
    )
    Use Spawn Probability=false
   )
  ]
  particleSpawnScript [
   /Niagara/Modules/Spawn/Initialization/V2/InitializeParticle(
    Color=<rgba 0.3636 0.9138 1 1>
    Lifetime Max=6
    Lifetime Min=5.5
    Lifetime Mode=::NewEnumerator1
    Mesh Scale Mode=::NewEnumerator3
    Sprite Size Mode=::NewEnumerator3
    Uniform Sprite Size=70
    UsePositionOffset=false
    Write Material Random=true
   )
   &AddVelocity(Velocity=<0 -85 0>)
   &ShapeLocation(
    Box Size=<100 0 80>
    MaskSpawnGroup=false
    Offset=&LerpPosition(
     A=<100 200 150>
     B=<-100 200 150>
     Alpha=&Lerp_Float(
      A=&RandomRangeFloat(Minimum=0, Maximum=0.04)
      B=&RandomRangeFloat(Minimum=0.96, Maximum=1)
      Alpha=&SelectFloatFromArray(
       Direct Array Index=&ModuloInt(A=$Engine.Emitter.TotalSpawnedParticles, B=2)
       Float Selection Array=DataInterface() {propertyValues {FloatData [0,1]}}
      )
     )
    )
    Offset Mode=::NewEnumerator0
    Override Seed=false
    Shape Primitive=::NewEnumerator2
    Surface Only=false
    UseOwnerScale=true
   )
  ]
  particleUpdateScript [
   &ParticleState()
   &ScaleSpriteSize(Uniform Curve Sprite Scale=DataInterface() {propertyValues {Curve {
       defaultValue=0
       keys [
        {interpMode=RCIM_Cubic, tangentMode=RCTM_User, time=0, value=0}
        {arriveTangentWeight=0.1167, interpMode=RCIM_Cubic, tangentMode=RCTM_User, time=0.35, value=1}
        {arriveTangentWeight=0.1, interpMode=RCIM_Cubic, tangentMode=RCTM_User, time=0.65, value=1}
        {arriveTangentWeight=0.1167, interpMode=RCIM_Cubic, tangentMode=RCTM_User, time=1, value=0}
       ]
      }}})
   &ScaleMeshSize(Scale Factor=&VectorFromCurve(VectorCurve=DataInterface() {propertyValues {
       XCurve {keys [
         {interpMode=RCIM_Cubic, tangentMode=RCTM_User, time=0, value=0}
         {interpMode=RCIM_Cubic, tangentMode=RCTM_User, time=0.35, value=1}
         {interpMode=RCIM_Cubic, tangentMode=RCTM_User, time=0.65, value=1}
         {interpMode=RCIM_Cubic, tangentMode=RCTM_User, time=1, value=0}
        ]}
       YCurve {keys [
         {interpMode=RCIM_Cubic, tangentMode=RCTM_User, time=0, value=0}
         {interpMode=RCIM_Cubic, tangentMode=RCTM_User, time=0.35, value=1}
         {interpMode=RCIM_Cubic, tangentMode=RCTM_User, time=0.65, value=1}
         {interpMode=RCIM_Cubic, tangentMode=RCTM_User, time=1, value=0}
        ]}
       ZCurve {keys [
         {interpMode=RCIM_Cubic, tangentMode=RCTM_User, time=0, value=0}
         {interpMode=RCIM_Cubic, tangentMode=RCTM_User, time=0.35, value=1}
         {interpMode=RCIM_Cubic, tangentMode=RCTM_User, time=0.65, value=1}
         {interpMode=RCIM_Cubic, tangentMode=RCTM_User, time=1, value=0}
        ]}
      }}))
   &CurlNoiseForce(Bias Noise=false, Noise Frequency=300, Pan Noise=false)
   &GravityForce(Gravity=&RandomRangeVector(
     Minimum=&VectorFromCurve(VectorCurve=DataInterface() {propertyValues {
        XCurve {keys [
          {interpMode=RCIM_Cubic, tangentMode=RCTM_User, time=0, value=0}
          {interpMode=RCIM_Cubic, tangentMode=RCTM_User, time=0.2, value=15}
          {interpMode=RCIM_Cubic, tangentMode=RCTM_User, time=0.8, value=-10}
          {interpMode=RCIM_Cubic, tangentMode=RCTM_User, time=1, value=0}
         ]}
        YCurve {keys [
          {interpMode=RCIM_Cubic, tangentMode=RCTM_User, time=0, value=0}
          {interpMode=RCIM_Cubic, tangentMode=RCTM_User, time=0.2, value=-10}
          {interpMode=RCIM_Cubic, tangentMode=RCTM_User, time=0.5, value=25}
          {interpMode=RCIM_Cubic, tangentMode=RCTM_User, time=1, value=0}
         ]}
        ZCurve {keys [
          {interpMode=RCIM_Cubic, tangentMode=RCTM_User, time=0, value=0}
          {interpMode=RCIM_Cubic, tangentMode=RCTM_User, time=0.2, value=30}
          {interpMode=RCIM_Cubic, tangentMode=RCTM_User, time=0.6, value=-20}
          {interpMode=RCIM_Cubic, tangentMode=RCTM_User, time=1, value=0}
         ]}
       }})
     Maximum=&VectorFromCurve(VectorCurve=DataInterface() {propertyValues {
        XCurve {keys [
          {interpMode=RCIM_Cubic, tangentMode=RCTM_User, time=0, value=0}
          {interpMode=RCIM_Cubic, tangentMode=RCTM_User, time=0.2, value=-15}
          {interpMode=RCIM_Cubic, tangentMode=RCTM_User, time=0.7, value=15}
          {interpMode=RCIM_Cubic, tangentMode=RCTM_User, time=1, value=0}
         ]}
        YCurve {
         defaultValue=0
         keys [
          {interpMode=RCIM_Cubic, tangentMode=RCTM_User, time=0, value=0}
          {interpMode=RCIM_Cubic, tangentMode=RCTM_User, time=0.3, value=20}
          {interpMode=RCIM_Cubic, tangentMode=RCTM_User, time=0.6, value=-10}
          {interpMode=RCIM_Cubic, tangentMode=RCTM_User, time=1, value=0}
         ]
        }
        ZCurve {keys [
          {interpMode=RCIM_Cubic, tangentMode=RCTM_User, time=0, value=0}
          {interpMode=RCIM_Cubic, tangentMode=RCTM_User, time=0.2, value=-35}
          {interpMode=RCIM_Cubic, tangentMode=RCTM_User, time=0.7, value=30}
          {interpMode=RCIM_Cubic, tangentMode=RCTM_User, time=1, value=0}
         ]}
       }})
    ))
   &MeshRotationForce(Rotation=&RandomRangeVector(Minimum=<0 0 0.2>, Maximum=<0 0 -0.2>)) {enabled=false}
   /SetVariables(Particles.Glow Position:NiagaraPosition=&AddVectorToPosition(Position=$Particles.Position, Vector=<0 0 50>)) {isSetParameters=true, module=SetVariables_#1}
   &SolveForcesAndVelocity(Clamp Velocity=false, Limit Acceleration=false)
   /Niagara/Modules/Spawn/Location/StaticMeshLocation(
    MaskSpawnGroup=false
    OffsetAlongNormal=false
    OverrideRandomSeed=false
    SamplePosOffset=false
    Static Mesh=DataInterface() {propertyValues {
      DefaultMesh=/BRCosmetics/Animation/Game/MainPlayer/Emotes/PartyJelly/FX/Jellyfish_Iterations/SM_Emote_PartyJelly_Jellyfish
      InstanceIndex=-1
      LODIndex=0
      PreviewMesh=&SM_Emote_PartyJelly_Jellyfish
      SoftSourceActor=null
      SourceMode=Default
      bAllowSamplingFromStreamingLODs=false
      bCaptureTransformsPerFrame=true
      bUsePhysicsBodyVelocity=false
      FilteredSockets []
      LODIndexUserParameter {parameter {name=None, typeDefHandle {registeredTypeIndex=108}, varData []}}
      MeshParameterBinding {parameter {name=None, typeDefHandle {registeredTypeIndex=1}, varData []}}
      SectionFilter {allowedMaterialSlots []}
     }}
   ) {enabled=false}
  ]
  rendererClasses [&NiagaraMeshRendererProperties, &NiagaraSpriteRendererProperties]
  renderers [
   {rendererClass=&NiagaraMeshRendererProperties, rendererIndex=0}
   {rendererClass=&NiagaraSpriteRendererProperties, rendererIndex=1}
  ]
 }
 Jellyfish001 {
  emitterUpdateScript [
   &EmitterState(Life Cycle Mode=::NewEnumerator1, Loop Delay=0.1, Loop Duration=1, UseLoopDelay=false)
   &SpawnRate(
    SpawnRate=&MakeCustomFloatFromBool(
     Bool=&SetBoolByFloatComparison(A=&VectorLength(Vector=$User.OwnerLocalVelocity), B=301)
     True Float=&RandomRangeFloat(Minimum=8, Maximum=11)
     False Float=&RandomRangeFloat(Minimum=2, Maximum=3)
    )
    Use Spawn Probability=false
   )
  ]
  particleSpawnScript [
   &PartitionParticles()
   &InitializeParticle(
    Color=<rgba 0.112 0.635 1 1>
    Lifetime Max=5.8
    Lifetime Min=5.4
    Lifetime Mode=::NewEnumerator1
    Mesh Scale Mode=::NewEnumerator3
    Mesh Uniform Scale=0.75
    Sprite Size Mode=::NewEnumerator3
    Uniform Sprite Size=50
    UsePositionOffset=false
    Write Material Random=true
   )
   &AddVelocity(Velocity=<0 -85 0.001>)
   &ShapeLocation(
    Box Size=<85 0 70>
    MaskSpawnGroup=false
    Offset=&LerpPosition(
     A=<100 200 150>
     B=<-100 200 150>
     Alpha=&Lerp_Float(
      A=&RandomRangeFloat(Minimum=0, Maximum=0.03)
      B=&RandomRangeFloat(Minimum=0.97, Maximum=1)
      Alpha=&SelectFloatFromArray(
       Direct Array Index=&ModuloInt(A=$Engine.Emitter.TotalSpawnedParticles, B=2)
       Float Selection Array=DataInterface() {propertyValues {FloatData [0,1]}}
      )
     )
    )
    Offset Mode=::NewEnumerator0
    Override Seed=false
    Shape Primitive=::NewEnumerator2
    Surface Only=false
    UseOwnerScale=true
   )
   &InitialMeshOrientation(Mesh Axis To Orient=<1 0 -1>, Mesh Orientation Mode=::NewEnumerator2, Orientation Vector=<0 0 -0.3>, Use Rotation Vector=false)
   /SetVariables(Particles.Glow Position:NiagaraPosition=&AddVectorToPosition(Position=$Particles.Position, Vector=<0 0 100>)) {enabled=false, isSetParameters=true, module=SetVariables_#1}
  ]
  particleUpdateScript [
   &ParticleState()
   &ScaleSpriteSize(Uniform Curve Sprite Scale=DataInterface() {propertyValues {Curve {
       defaultValue=0
       keys [
        {interpMode=RCIM_Cubic, tangentMode=RCTM_User, time=0, value=0}
        {arriveTangentWeight=0.1167, interpMode=RCIM_Cubic, tangentMode=RCTM_User, time=0.35, value=1}
        {arriveTangentWeight=0.1, interpMode=RCIM_Cubic, tangentMode=RCTM_User, time=0.65, value=1}
        {arriveTangentWeight=0.1167, interpMode=RCIM_Cubic, tangentMode=RCTM_User, time=1, value=0}
       ]
      }}})
   &ScaleMeshSize(Scale Factor=&VectorFromCurve(VectorCurve=DataInterface() {propertyValues {
       XCurve {keys [
         {interpMode=RCIM_Cubic, tangentMode=RCTM_User, time=0, value=0}
         {interpMode=RCIM_Cubic, tangentMode=RCTM_User, time=0.35, value=1}
         {interpMode=RCIM_Cubic, tangentMode=RCTM_User, time=0.65, value=1}
         {interpMode=RCIM_Cubic, tangentMode=RCTM_User, time=1, value=0}
        ]}
       YCurve {keys [
         {interpMode=RCIM_Cubic, tangentMode=RCTM_User, time=0, value=0}
         {interpMode=RCIM_Cubic, tangentMode=RCTM_User, time=0.35, value=1}
         {interpMode=RCIM_Cubic, tangentMode=RCTM_User, time=0.65, value=1}
         {interpMode=RCIM_Cubic, tangentMode=RCTM_User, time=1, value=0}
        ]}
       ZCurve {keys [
         {interpMode=RCIM_Cubic, tangentMode=RCTM_User, time=0, value=0}
         {interpMode=RCIM_Cubic, tangentMode=RCTM_User, time=0.35, value=1}
         {interpMode=RCIM_Cubic, tangentMode=RCTM_User, time=0.65, value=1}
         {interpMode=RCIM_Cubic, tangentMode=RCTM_User, time=1, value=0}
        ]}
      }}))
   &CurlNoiseForce(
    Bias Noise=false
    Noise Frequency=100
    Noise Strength=&MakeFloatFromVector(Vector=<0 0 10>, Channel=::NewEnumerator0)
    Pan Noise=false
   )
   &GravityForce(Gravity=&RandomRangeVector(
     Minimum=&VectorFromCurve(VectorCurve=DataInterface() {propertyValues {
        XCurve {keys [<at 0 0>,<at 1 0>]}
        YCurve {keys [<at 0 0>,<at 1 0>]}
        ZCurve {keys [
          {interpMode=RCIM_Cubic, tangentMode=RCTM_User, time=0, value=0}
          {interpMode=RCIM_Cubic, tangentMode=RCTM_User, time=0.25, value=20}
          {interpMode=RCIM_Cubic, tangentMode=RCTM_User, time=0.55, value=-28}
          {interpMode=RCIM_Cubic, tangentMode=RCTM_User, time=0.8, value=0}
         ]}
       }})
     Maximum=&VectorFromCurve(VectorCurve=DataInterface() {propertyValues {
        XCurve {defaultValue=1, keys [<at 0 0>,<at 1 0>]}
        YCurve {defaultValue=1, keys [<at 0 0>,<at 1 0>]}
        ZCurve {keys [
          {interpMode=RCIM_Cubic, leaveTangentWeight=0.3333, tangentMode=RCTM_User, time=0, value=0}
          {arriveTangentWeight=0.0833, interpMode=RCIM_Cubic, leaveTangentWeight=0.1833, tangentMode=RCTM_User, time=0.25, value=-28}
          {arriveTangentWeight=0.1, interpMode=RCIM_Cubic, leaveTangentWeight=0.0833, tangentMode=RCTM_User, time=0.55, value=20}
          {arriveTangentWeight=0.2667, interpMode=RCIM_Cubic, leaveTangentWeight=0.0667, tangentMode=RCTM_User, time=0.8, value=0}
          <at 1 0>
         ]}
       }})
    )) {module=GravityForce001}
   &GravityForce(Gravity=&RandomRangeVector(
     Minimum=&VectorFromCurve(VectorCurve=DataInterface() {propertyValues {
        XCurve {keys [
          {interpMode=RCIM_Cubic, tangentMode=RCTM_User, time=0, value=0}
          {interpMode=RCIM_Cubic, tangentMode=RCTM_User, time=0.2, value=15}
          {interpMode=RCIM_Cubic, tangentMode=RCTM_User, time=0.6, value=-10}
          {interpMode=RCIM_Cubic, tangentMode=RCTM_User, time=0.8, value=0}
         ]}
        YCurve {keys [
          {interpMode=RCIM_Cubic, tangentMode=RCTM_User, time=0, value=0}
          {interpMode=RCIM_Cubic, tangentMode=RCTM_User, time=0.2, value=0}
          {interpMode=RCIM_Cubic, tangentMode=RCTM_User, time=0.5, value=25}
          {interpMode=RCIM_Cubic, tangentMode=RCTM_User, time=0.8, value=0}
         ]}
        ZCurve {keys [
          {interpMode=RCIM_Cubic, tangentMode=RCTM_User, time=0, value=0}
          {interpMode=RCIM_Cubic, tangentMode=RCTM_User, time=0.2, value=20}
          {interpMode=RCIM_Cubic, tangentMode=RCTM_User, time=0.6, value=-7.5}
          {interpMode=RCIM_Cubic, tangentMode=RCTM_User, time=0.8, value=0}
          <at 1 -5 RCIM_Cubic>
         ]}
       }})
     Maximum=&VectorFromCurve(VectorCurve=DataInterface() {propertyValues {
        XCurve {keys [
          {interpMode=RCIM_Cubic, tangentMode=RCTM_User, time=0, value=0}
          {interpMode=RCIM_Cubic, tangentMode=RCTM_User, time=0.2, value=-15}
          {interpMode=RCIM_Cubic, tangentMode=RCTM_User, time=0.6, value=15}
          {interpMode=RCIM_Cubic, tangentMode=RCTM_User, time=0.8, value=0}
         ]}
        YCurve {
         defaultValue=0
         keys [
          {interpMode=RCIM_Cubic, tangentMode=RCTM_User, time=0, value=0}
          {interpMode=RCIM_Cubic, tangentMode=RCTM_User, time=0.25, value=20}
          {interpMode=RCIM_Cubic, tangentMode=RCTM_User, time=0.6, value=0}
          {interpMode=RCIM_Cubic, tangentMode=RCTM_User, time=0.8, value=0}
         ]
        }
        ZCurve {keys [
          {interpMode=RCIM_Cubic, tangentMode=RCTM_User, time=0, value=0}
          {interpMode=RCIM_Cubic, tangentMode=RCTM_User, time=0.2, value=-20}
          {interpMode=RCIM_Cubic, tangentMode=RCTM_User, time=0.6, value=10}
          {interpMode=RCIM_Cubic, tangentMode=RCTM_User, time=0.8, value=0}
          <at 1 -5 RCIM_Cubic>
         ]}
       }})
    )) {enabled=false}
   &MeshRotationForce(Rotation=&RandomRangeVector(Minimum=<0.05 0.05 0.2>, Maximum=<-0.05 0.01 -0.2>))
   &SolveForcesAndVelocity(Clamp Velocity=false, Limit Acceleration=false)
  ]
  rendererClasses [&NiagaraMeshRendererProperties, &NiagaraSpriteRendererProperties]
  renderers [
   {rendererClass=&NiagaraMeshRendererProperties, rendererIndex=0}
   {rendererClass=&NiagaraSpriteRendererProperties, rendererIndex=1}
  ]
 }
}
summary {
 systemName=NS_Emote_PartyJelly_Bubbles1
 emitters [
  {emitterName=Bubbles, rendererClasses [&NiagaraSpriteRendererProperties]}
  {
   bEnabled=false
   emitterName=Jellyfish
   rendererClasses [&NiagaraMeshRendererProperties, &NiagaraSpriteRendererProperties]
  }
  {emitterName=Jellyfish001, rendererClasses [&NiagaraMeshRendererProperties, &NiagaraSpriteRendererProperties]}
  {bEnabled=false, emitterName=FN_Debug, rendererClasses [&NiagaraSpriteRendererProperties]}
 ]
 userVariables [
  User.OwnerLocalVelocity {type=Vector3f, description="", defaultValue {struct=Vector3f, value=<0 0 0>}}
  User.Skeletal Mesh {
   type=NiagaraDataInterfaceSkeletalMesh
   defaultValue {
    struct=NiagaraExt_VariableValue_DataInterface
    value {dataInterface=&self:NiagaraDataInterfaceSkeletalMesh_0}
   }
  }
 ]
}
